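# Level Sequence for animating the BerryFlight_F Season 41 custom skin's main-player Battle Royale menu animation on a FortPlayerMannequin.
# The mannequin binding has spawn and transform tracks, a 320-frame skeletal animation section, and Control Rig tracks for MR_Biped_v02 and Fortnite_3L_Face_CtrlRig.
# The face rig track carries dense, near per-frame keys on eye-aim and chin controls, while many other controls hold constant values.
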LevelSequence /BRCosmetics/Animation/Game/MainPlayer/Menu/BR/CustomSkins/_S41/BerryFlight_F/Animation/BerryFlight_Anim_LevelSequence
bindings [FortPlayerMannequin {
  id=#1
  possessed_class=FortPlayerMannequin
  tracks [
   {type=MovieSceneSpawnTrack, sections [{type=MovieSceneSpawnSection, channels [None {type=MovieSceneScriptingBoolChannel, default=true}]}]}
   {
    type=MovieScene3DTransformTrack
    sections [{
      type=MovieScene3DTransformSection
      channels [
       Location.X {type=MovieSceneScriptingDoubleChannel, default=0}
       Location.Y {type=MovieSceneScriptingDoubleChannel, default=0}
       Location.Z {type=MovieSceneScriptingDoubleChannel, default=0}
       Rotation.X {type=MovieSceneScriptingDoubleChannel, default=0}
       Rotation.Y {type=MovieSceneScriptingDoubleChannel, default=0}
       Rotation.Z {type=MovieSceneScriptingDoubleChannel, default=0}
       Scale.X {type=MovieSceneScriptingDoubleChannel, default=1}
       Scale.Y {type=MovieSceneScriptingDoubleChannel, default=1}
       Scale.Z {type=MovieSceneScriptingDoubleChannel, default=1}
      ]
     }]
   }
   {type=MovieSceneSkeletalAnimationTrack, sections []}
   {type=MovieSceneSkeletalAnimationTrack, sections [{type=MovieSceneSkeletalAnimationSection, end_frame=320, start_frame=0}]}
   {type=MovieSceneControlRigParameterTrack, display_name=MR_Biped_v02, sections [{type=MovieSceneControlRigParameterSection}]}
   {
    type=MovieSceneControlRigParameterTrack
    display_name=Fortnite_3L_Face_CtrlRig
    sections [{
      type=MovieSceneControlRigParameterSection
      channels [
       CTRL_C_eyesAim.Location.X {
        type=MovieSceneScriptingFloatChannel
        default=0
        thinned_from=321
        keys [[0,-3.0916],[1,-3.0784],[3,-3.0211],[5,-2.9432],[9,-2.7684],[11,-2.661],[12,-2.613],[13,-2.5715],[14,-2.5371],[15,-2.5111],[16,-2.494],[17,-2.4861],[18,-2.8534],[19,-2.9649],[20,-3.0163],[21,-3.0514],[22,-3.0979],[23,-3.1565],[28,-3.5071],[30,-3.624],[33,-3.7724],[37,-3.9289],[38,-4.1576],[39,-4.2405],[40,-4.2789],[45,-4.4001],[47,-4.4332],[48,-3.9606],[49,-3.8295],[50,-3.8012],[54,-3.7916],[56,-3.7679],[59,-3.699],[60,-3.6858],[61,-4.3012],[62,-4.5077],[63,-4.5698],[64,-4.5489],[65,-4.5028],[66,-4.4805],[67,-4.4373],[68,-4.2051],[69,-4.1243],[70,-4.0853],[71,-4.0621],[72,-4.0088],[73,-3.9393],[74,-3.8554],[77,-3.5608],[78,-3.4694],[79,-3.4049],[80,-3.3705],[81,-3.3707],[82,-3.3945],[83,-3.429],[87,-3.6019],[88,-3.6376],[89,-3.865],[90,-3.9392],[91,-3.9916],[92,-4.0211],[93,-4.03],[103,-4.0224],[104,-4.0171],[105,-4.5059],[106,-4.6562],[107,-4.7074],[108,-4.7285],[111,-4.7344],[112,-4.7463],[113,-4.7763],[114,-5.09],[115,-5.1424],[118,-5.0684],[120,-5.0488],[122,-5.0586],[124,-5.1005],[126,-5.1677],[128,-5.2615],[130,-5.3906],[132,-5.5502],[133,-5.2449],[134,-5.1618],[135,-5.1666],[136,-5.2852],[138,-5.713],[139,-5.8414],[141,-5.9433],[142,-5.9169],[143,-5.8775],[144,-5.6363],[145,-4.9529],[146,-3.3579],[147,-3.387],[148,-3.4797],[149,-3.6142],[150,-3.756],[151,-3.8835],[152,-4.0024],[156,-4.4278],[157,-4.525],[158,-4.611],[159,-4.6798],[160,-4.736],[161,-4.7768],[163,-4.8225],[165,-4.8244],[166,-4.8178],[167,-4.8014],[170,-4.7161],[171,-4.6728],[172,-4.6011],[173,-4.4904],[174,-3.9241],[175,-3.7662],[176,-3.6234],[178,-3.2476],[179,-3.089],[180,-2.9732],[181,-2.8692],[182,-5.0825],[183,-2.4801],[184,-1.9162],[185,-1.8292],[186,-2.3055],[187,-2.2105],[188,-2.1055],[190,-1.8745],[191,-2.1349],[192,-2.1899],[193,-2.1646],[194,-2.0924],[195,-2.0272],[196,-1.9704],[197,-1.9238],[198,-1.888],[199,-1.8631],[200,-1.8495],[201,-1.8398],[202,-1.8248],[203,-1.805],[204,-1.7784],[205,-1.7395],[206,-1.6947],[207,-1.6441],[208,-2.0301],[209,-2.0965],[210,-2.0625],[215,-1.7643],[217,-1.6353],[218,-1.1632],[219,-0.9866],[220,-0.9121],[221,-2.0126],[222,-1.9855],[225,-1.9855],[227,-1.9964],[230,-1.9964],[231,-2.0034],[232,-2.0518],[233,-2.1423],[234,-2.2674],[235,-2.42],[236,-2.5857],[237,-2.7398],[238,-2.8643],[239,-2.4034],[240,-11.1006],[241,-7.1708],[242,-5.69],[243,-5.2816],[244,-5.3047],[245,-5.2246],[246,-5.0415],[247,-5.0826],[248,-5.1518],[249,-5.2349],[250,-5.3495],[251,-5.3769],[252,-5.3621],[254,-5.2864],[264,-5.0883],[267,-4.9967],[268,-4.9587],[269,-4.4424],[270,-4.2559],[271,-4.1595],[273,-4.0109],[282,-3.4249],[284,-3.3123],[285,-3.2645],[286,-3.2247],[287,-3.195],[288,-3.5709],[289,-3.6424],[290,-3.6455],[291,-3.633],[293,-3.624],[295,-3.6254],[296,-3.605],[297,-3.4626],[298,-3.4286],[299,-3.4135],[301,-3.4041],[305,-3.4316],[306,-3.1428],[307,-3.0942],[308,-3.0574],[310,-3.0544],[317,-3.1177],[319,-3.1281],[320,-3.1281]]
       }
       CTRL_C_eyesAim.Location.Y {
        type=MovieSceneScriptingFloatChannel
        default=0
        thinned_from=321
        keys [[0,12.9252],[17,12.9283],[18,13.0196],[37,13.1528],[47,13.1771],[48,13.0957],[59,13.1824],[60,13.2218],[61,13.4442],[62,13.5241],[65,13.607],[66,13.6964],[67,14.8173],[68,15.748],[69,16.536],[70,16.9824],[71,17.1938],[72,17.2078],[73,17.1621],[76,16.8328],[78,16.5413],[80,16.1905],[81,16.0447],[83,15.8447],[86,15.6677],[88,15.6163],[89,15.363],[91,15.2863],[101,15.3953],[104,15.4839],[105,15.7871],[106,15.9118],[113,16.2557],[114,16.1579],[128,16.452],[130,16.5478],[132,16.7276],[133,15.7132],[134,15.6185],[135,15.6662],[136,15.5458],[137,15.3925],[138,15.2944],[139,15.318],[141,15.5616],[142,15.3474],[143,15.7582],[144,16.8547],[145,20.8809],[146,13.7085],[147,13.7505],[148,13.9307],[149,14.215],[150,14.5309],[151,14.7808],[153,15.1831],[154,15.3413],[155,15.4686],[156,15.5594],[161,15.7078],[168,15.7196],[169,15.761],[172,15.9676],[173,15.9784],[174,15.6889],[175,15.6392],[177,15.6174],[178,15.6353],[179,15.7232],[180,15.9865],[181,16.1116],[182,15.6532],[183,14.7486],[184,12.9488],[196,13.7353],[204,13.9955],[217,14.0016],[220,14.0451],[221,13.9874],[230,13.9879],[231,13.9557],[233,13.8111],[235,13.603],[236,13.4588],[239,12.9404],[240,17.1429],[241,14.6998],[242,14.248],[243,14.0765],[244,13.7797],[245,13.4523],[246,13.2616],[248,13.155],[268,12.9425],[269,12.8962],[287,12.984],[289,12.9143],[296,12.9651],[297,12.8684],[302,12.8932],[305,12.9557],[306,12.8745],[307,12.9007],[320,12.9801]]
       }
       CTRL_C_eyesAim.Location.Z {
        type=MovieSceneScriptingFloatChannel
        default=0
        thinned_from=321
        keys [[0,10.063],[17,9.8887],[18,9.3962],[19,9.2427],[20,9.1976],[37,9.0281],[38,9.2077],[39,9.2833],[40,9.3061],[47,9.1562],[48,9.3586],[49,9.3962],[50,9.381],[51,9.3351],[55,8.9997],[58,8.8453],[59,8.7794],[60,8.5803],[61,7.8288],[62,7.5932],[63,7.4421],[64,7.3445],[65,7.2709],[66,6.9351],[67,3.7805],[68,1.8286],[69,0.4749],[70,-0.2119],[71,-0.5212],[72,-0.5492],[73,-0.4945],[74,-0.3464],[75,-0.193],[76,-0.052],[77,0.1402],[78,0.3712],[79,0.636],[80,0.9262],[81,1.1722],[82,1.3664],[83,1.528],[84,1.6606],[85,1.7751],[86,1.866],[87,1.9301],[88,1.9729],[89,2.4967],[90,2.6119],[91,2.6732],[92,2.6955],[93,2.6982],[94,2.6911],[96,2.6586],[97,2.6325],[98,2.5995],[99,2.5598],[100,2.5143],[102,2.4079],[104,2.2885],[105,1.8205],[106,1.6294],[107,1.5171],[108,1.4213],[109,1.3311],[110,1.2484],[111,1.1752],[112,1.1125],[113,1.0621],[114,1.3054],[115,1.3393],[116,1.3094],[117,1.2654],[118,1.2245],[120,1.1518],[125,0.9826],[126,0.9452],[127,0.9046],[128,0.8576],[129,0.8066],[130,0.7367],[131,0.6366],[132,0.495],[133,2.1486],[134,2.3044],[135,2.2146],[136,2.4823],[137,2.8568],[138,3.1362],[139,3.1344],[140,2.894],[141,2.6747],[142,3.1017],[143,2.2684],[144,0.3218],[145,-4.5421],[146,6.5578],[147,6.4242],[148,5.8746],[150,4.301],[151,3.7354],[152,3.2849],[153,2.908],[154,2.6084],[155,2.3806],[156,2.2284],[157,2.146],[158,2.1073],[161,2.0311],[162,2.0244],[163,2.029],[165,2.0459],[166,2.048],[167,2.0378],[168,2.0085],[169,1.9247],[170,1.7816],[171,1.6305],[172,1.5177],[173,1.4743],[174,1.8849],[175,1.9504],[176,1.9366],[177,1.942],[178,1.8855],[179,1.7071],[180,1.2308],[181,1.0091],[182,2.2154],[183,3.6247],[184,9.7084],[185,9.2999],[186,9.0182],[188,8.3598],[190,7.7393],[193,6.9588],[194,6.7139],[195,6.4992],[196,6.3178],[197,6.1702],[198,6.0479],[203,5.585],[204,5.5218],[207,5.5083],[208,5.5475],[217,5.5074],[218,5.4613],[220,5.4376],[221,5.5457],[230,5.544],[231,5.638],[232,5.8361],[233,6.08],[234,6.3893],[235,6.7826],[236,7.3241],[237,8.0405],[238,8.8975],[239,9.7971],[240,2.7894],[241,5.2586],[242,5.6974],[243,6.0204],[244,7.0145],[245,8.2772],[246,9.0904],[247,9.4804],[248,9.8002],[249,10.0255],[250,10.2008],[251,10.307],[252,10.2775],[253,10.3496],[256,10.7746],[258,11.009],[260,11.1825],[262,11.2997],[264,11.3626],[266,11.3677],[268,11.3109],[269,11.1688],[275,10.8494],[279,10.547],[284,10.0528],[287,9.7091],[288,10.1363],[289,10.3878],[290,10.4094],[291,10.3641],[295,9.9762],[296,10.0004],[297,10.6574],[298,10.7823],[299,10.7644],[301,10.5647],[304,10.1173],[305,9.9857],[306,10.4565],[307,10.2386],[308,10.3517],[310,10.2217],[314,9.8943],[316,9.786],[318,9.7245],[320,9.7114]]
       }
       CTRL_C_eyesAim.Rotation.X {
        type=MovieSceneScriptingFloatChannel
        default=0
        thinned_from=321
        keys [[0,-1.7881],[2,-1.8171],[5,-1.8718],[8,-1.9457],[9,-1.9729],[12,-2.0641],[16,-2.1695],[17,-2.201],[18,-3.3733],[19,-3.7395],[20,-3.8474],[23,-3.9528],[27,-4.0474],[34,-4.1792],[37,-4.2556],[38,-3.8272],[39,-3.6469],[40,-3.5926],[42,-3.6653],[44,-3.7652],[47,-3.9514],[48,-3.466],[49,-3.3761],[50,-3.4122],[51,-3.5216],[52,-3.6908],[53,-3.8965],[54,-4.1152],[55,-4.3229],[56,-4.4749],[58,-4.6923],[59,-4.8502],[60,-5.3276],[61,-7.1431],[62,-7.7162],[63,-8.0841],[64,-8.3212],[65,-8.5],[66,-9.3194],[67,-17.2107],[68,-22.3145],[69,-26.0001],[70,-27.9252],[71,-28.8055],[72,-28.8809],[73,-28.7176],[76,-27.4369],[77,-26.8897],[78,-26.2389],[79,-25.5019],[80,-24.7032],[81,-24.0335],[82,-23.5091],[83,-23.0756],[84,-22.722],[85,-22.4178],[86,-22.1775],[87,-22.0088],[88,-21.8969],[89,-20.5221],[90,-20.2229],[91,-20.0646],[92,-20.0076],[94,-20.0195],[97,-20.173],[100,-20.484],[104,-21.0795],[105,-22.3548],[106,-22.8787],[108,-23.4471],[111,-24.1176],[113,-24.4304],[114,-23.7893],[115,-23.7008],[127,-24.8993],[129,-25.1803],[130,-25.3804],[131,-25.6664],[132,-26.0701],[133,-21.5238],[134,-21.1031],[135,-21.3426],[136,-20.6383],[137,-19.6615],[138,-18.94],[139,-18.9519],[140,-19.5869],[141,-20.1691],[142,-19.0417],[143,-21.2463],[144,-26.5658],[145,-41.5725],[146,-10.2272],[147,-10.5552],[148,-11.9111],[150,-15.8583],[151,-17.3054],[152,-18.471],[153,-19.4558],[154,-20.2459],[155,-20.8524],[156,-21.2613],[157,-21.486],[161,-21.8079],[166,-21.7654],[167,-21.7916],[168,-21.8682],[169,-22.0908],[171,-22.877],[172,-23.1769],[173,-23.2868],[174,-22.1473],[175,-21.9637],[176,-21.9927],[177,-21.9684],[178,-22.1092],[179,-22.5768],[180,-23.8515],[181,-24.4471],[182,-21.3352],[183,-17.541],[184,-2.6284],[185,-3.5998],[186,-4.2731],[187,-5.0672],[188,-5.8482],[189,-6.6051],[190,-7.3396],[191,-7.9564],[193,-9.233],[194,-9.8287],[195,-10.3522],[196,-10.7955],[197,-11.1571],[198,-11.4569],[199,-11.7108],[203,-12.5972],[204,-12.7531],[207,-12.7842],[208,-12.6939],[217,-12.7862],[218,-12.8923],[220,-12.9467],[221,-12.6981],[230,-12.7019],[231,-12.4696],[232,-11.9812],[233,-11.3818],[234,-10.6247],[235,-9.6652],[236,-8.35],[237,-6.6197],[238,-4.5637],[239,-2.4186],[240,-20.3243],[241,-13.5456],[242,-12.3992],[243,-11.586],[244,-9.1365],[245,-6.0673],[246,-4.1114],[247,-3.1791],[248,-2.4167],[249,-1.8804],[250,-1.4641],[251,-1.2119],[252,-1.2819],[253,-1.1108],[254,-0.7964],[255,-0.4368],[256,-0.1035],[257,0.1924],[258,0.4511],[259,0.673],[260,0.8611],[261,1.0159],[262,1.1379],[263,1.2288],[264,1.2863],[265,1.3099],[266,1.2983],[267,1.2368],[268,1.1642],[269,0.8286],[270,0.6801],[271,0.5745],[272,0.466],[273,0.3476],[274,0.2173],[275,0.0739],[276,-0.0844],[277,-0.2571],[278,-0.4434],[279,-0.6414],[280,-0.8518],[281,-1.074],[282,-1.3087],[283,-1.5562],[284,-1.8127],[285,-2.0769],[286,-2.3502],[287,-2.6291],[288,-1.6147],[289,-1.0185],[290,-0.9673],[291,-1.0746],[292,-1.3158],[293,-1.5383],[294,-1.7765],[295,-1.9951],[296,-1.9374],[297,-0.38],[298,-0.0847],[299,-0.1269],[300,-0.3412],[301,-0.5993],[302,-0.9208],[303,-1.2884],[304,-1.6598],[305,-1.9723],[306,-0.8554],[307,-1.3718],[308,-1.1037],[309,-1.2367],[310,-1.4118],[311,-1.6125],[312,-1.8206],[313,-2.0175],[314,-2.1887],[315,-2.3317],[316,-2.446],[317,-2.5325],[318,-2.5924],[319,-2.6235],[320,-2.6235]]
       }
       CTRL_C_eyesAim.Rotation.Y {
        type=MovieSceneScriptingFloatChannel
        default=0
        thinned_from=321
        keys [[0,-6.5435],[17,-6.493],[18,-6.5215],[37,-6.7078],[38,-6.7315],[47,-6.786],[48,-6.6844],[49,-6.6613],[60,-6.6874],[61,-6.9047],[62,-6.9944],[63,-7.0322],[65,-7.0266],[66,-7.0508],[67,-7.213],[68,-7.0172],[69,-6.8678],[70,-6.7725],[72,-6.6556],[74,-6.5056],[76,-6.3248],[77,-6.2471],[78,-6.1833],[79,-6.1545],[80,-6.1606],[81,-6.1939],[87,-6.4902],[88,-6.5253],[89,-6.7561],[90,-6.8228],[91,-6.8678],[92,-6.8925],[104,-6.8724],[105,-7.287],[106,-7.424],[107,-7.4722],[109,-7.499],[111,-7.5002],[112,-7.5121],[113,-7.5418],[114,-7.841],[115,-7.8899],[119,-7.8115],[121,-7.8084],[124,-7.863],[127,-7.9783],[129,-8.096],[131,-8.2567],[132,-8.3537],[133,-7.9347],[134,-7.8509],[135,-7.8612],[136,-7.9431],[138,-8.2215],[139,-8.324],[141,-8.4785],[142,-8.3895],[143,-8.4813],[144,-8.459],[145,-7.5737],[146,-6.6078],[147,-6.6169],[148,-6.6455],[150,-6.7558],[154,-7.0475],[157,-7.3045],[159,-7.4403],[161,-7.5268],[163,-7.5673],[166,-7.5626],[170,-7.4773],[171,-7.4393],[172,-7.373],[173,-7.2693],[174,-6.7701],[176,-6.5096],[179,-6.0103],[180,-5.8283],[181,-5.6902],[182,-7.7877],[183,-5.8354],[184,-6.4285],[185,-6.3756],[186,-6.4215],[188,-6.3223],[190,-6.1854],[191,-6.2297],[192,-6.215],[193,-6.1728],[196,-6.0016],[198,-5.9194],[207,-5.6978],[208,-5.8988],[209,-5.9333],[210,-5.9157],[217,-5.6932],[218,-5.4453],[219,-5.3521],[220,-5.3126],[221,-5.8898],[222,-5.8757],[230,-5.8814],[231,-5.9014],[232,-5.9586],[233,-6.0392],[236,-6.3591],[237,-6.4539],[238,-6.5144],[239,-6.4816],[240,-13.253],[241,-8.7538],[242,-7.7876],[243,-7.5211],[244,-7.3684],[245,-7.1092],[246,-6.9102],[248,-6.7832],[250,-6.7124],[251,-6.6886],[252,-6.6947],[253,-6.6742],[258,-6.5071],[262,-6.4361],[266,-6.422],[268,-6.437],[269,-6.4752],[287,-6.5621],[288,-6.5763],[289,-6.5615],[295,-6.5924],[296,-6.5888],[297,-6.5338],[299,-6.5253],[305,-6.5745],[306,-6.5346],[320,-6.5546]]
       }
       CTRL_C_eyesAim.Rotation.Z {
        type=MovieSceneScriptingFloatChannel
        default=0
        thinned_from=321
        keys [[0,-1.8968],[1,-1.8643],[2,-1.7997],[3,-1.7262],[4,-1.6393],[5,-1.5391],[8,-1.2213],[9,-1.1174],[10,-0.9852],[11,-0.8585],[12,-0.7427],[13,-0.6423],[14,-0.5587],[15,-0.4948],[16,-0.4517],[17,-0.4297],[18,-1.1633],[19,-1.3848],[20,-1.4937],[21,-1.5718],[22,-1.6768],[23,-1.8108],[24,-1.9639],[27,-2.4644],[28,-2.6208],[29,-2.7613],[30,-2.8912],[31,-3.013],[32,-3.1267],[33,-3.233],[34,-3.332],[36,-3.5091],[37,-3.5893],[38,-4.1744],[39,-4.3895],[40,-4.486],[42,-4.6049],[44,-4.704],[46,-4.78],[47,-4.8085],[48,-3.7519],[49,-3.4541],[50,-3.3835],[52,-3.3444],[53,-3.3175],[54,-3.2827],[55,-3.2394],[56,-3.1872],[57,-3.1271],[58,-3.0591],[59,-2.9839],[60,-2.9003],[61,-4.1387],[62,-4.5567],[63,-4.66],[64,-4.5845],[65,-4.4565],[66,-4.3124],[67,-3.3525],[68,-2.3322],[69,-1.8327],[70,-1.595],[72,-1.3631],[73,-1.2315],[74,-1.089],[75,-0.9114],[76,-0.7373],[77,-0.5739],[78,-0.4277],[79,-0.3438],[80,-0.3287],[81,-0.3792],[82,-0.4705],[83,-0.5791],[84,-0.7005],[85,-0.8194],[86,-0.935],[87,-1.0424],[88,-1.13],[89,-1.7508],[90,-1.9424],[91,-2.073],[92,-2.1437],[93,-2.1641],[94,-2.1635],[99,-2.1246],[101,-2.1008],[102,-2.0836],[103,-2.0613],[104,-2.0348],[105,-2.9847],[106,-3.2609],[107,-3.3416],[108,-3.3617],[109,-3.3496],[110,-3.3243],[111,-3.3083],[112,-3.3172],[113,-3.368],[114,-4.1094],[115,-4.2322],[116,-4.1766],[117,-4.1012],[118,-4.04],[119,-4.0001],[120,-3.9773],[121,-3.9681],[122,-3.9794],[123,-4.0091],[124,-4.0501],[125,-4.1053],[126,-4.1735],[127,-4.2539],[128,-4.3486],[129,-4.4625],[132,-4.8543],[133,-4.6901],[134,-4.5531],[135,-4.5376],[136,-4.877],[137,-5.4584],[138,-6.0255],[139,-6.312],[140,-6.3593],[141,-6.3925],[142,-6.4703],[143,-6.1168],[144,-4.9831],[145,-2.3087],[146,-1.6097],[147,-1.6429],[148,-1.717],[149,-1.8263],[150,-1.9511],[151,-2.0966],[152,-2.2495],[153,-2.3952],[154,-2.5506],[155,-2.7307],[157,-3.1119],[158,-3.2898],[159,-3.4335],[160,-3.549],[161,-3.632],[162,-3.6893],[163,-3.7313],[164,-3.7462],[165,-3.7401],[166,-3.7262],[167,-3.6876],[168,-3.6212],[169,-3.5354],[170,-3.4317],[171,-3.2972],[172,-3.1118],[173,-2.8612],[174,-1.7342],[175,-1.4053],[176,-1.0909],[177,-0.6797],[178,-0.2608],[179,0.1207],[180,0.4634],[181,0.7275],[182,-4.3533],[183,1.076],[184,0.9542],[185,1.2626],[186,0.2181],[187,0.5248],[188,0.852],[189,1.1999],[190,1.5432],[191,0.9981],[192,0.9341],[193,1.0553],[194,1.2802],[195,1.481],[196,1.6536],[197,1.7947],[198,1.9044],[199,1.9848],[200,2.0364],[201,2.0786],[202,2.1322],[203,2.1967],[204,2.2713],[205,2.3614],[206,2.4651],[207,2.5822],[208,1.6885],[209,1.5347],[210,1.6133],[215,2.304],[216,2.45],[217,2.6026],[218,3.6952],[219,4.1038],[220,4.2763],[221,1.7289],[222,1.7917],[225,1.7917],[227,1.7665],[230,1.7665],[231,1.73],[232,1.5748],[233,1.3114],[234,0.9515],[235,0.5054],[236,-0.011],[237,-0.5492],[238,-1.0595],[239,-0.2118],[240,-18.0753],[241,-10.0462],[242,-6.7602],[243,-5.9119],[244,-6.2543],[245,-6.4268],[246,-6.2224],[247,-6.4264],[249,-6.9346],[250,-7.2531],[251,-7.3467],[252,-7.3035],[254,-7.1794],[255,-7.1512],[260,-7.0823],[262,-7.0297],[264,-6.9427],[266,-6.8072],[268,-6.6228],[269,-5.3668],[270,-4.9106],[271,-4.6717],[272,-4.4752],[273,-4.2965],[276,-3.7984],[278,-3.4518],[281,-2.9093],[282,-2.7331],[283,-2.5667],[284,-2.4126],[285,-2.2709],[286,-2.1472],[287,-2.0467],[288,-3.0424],[289,-3.2771],[290,-3.29],[292,-3.2098],[293,-3.1756],[295,-3.1282],[296,-3.0867],[297,-2.9253],[298,-2.8782],[300,-2.7982],[301,-2.7633],[302,-2.7361],[305,-2.6753],[306,-2.1206],[307,-1.9491],[308,-1.8924],[309,-1.8676],[310,-1.8512],[311,-1.8429],[312,-1.8416],[314,-1.8527],[316,-1.8703],[318,-1.881],[319,-1.8903],[320,-1.8903]]
       }
       CTRL_C_eyesAim.Scale.X {type=MovieSceneScriptingFloatChannel, default=1, const [1,321]}
       CTRL_C_eyesAim.Scale.Y {type=MovieSceneScriptingFloatChannel, default=1, const [1,321]}
       CTRL_C_eyesAim.Scale.Z {type=MovieSceneScriptingFloatChannel, default=1, const [1,321]}
       CTRL_L_eyeAim.Location.X {type=MovieSceneScriptingFloatChannel, default=0, const [0,321]}
       CTRL_L_eyeAim.Location.Y {type=MovieSceneScriptingFloatChannel, default=0, const [0,321]}
       CTRL_L_eyeAim.Location.Z {type=MovieSceneScriptingFloatChannel, default=0, const [0,321]}
       CTRL_L_eyeAim.Rotation.X {type=MovieSceneScriptingFloatChannel, default=0, const [0,321]}
       CTRL_L_eyeAim.Rotation.Y {type=MovieSceneScriptingFloatChannel, default=0, const [0,321]}
       CTRL_L_eyeAim.Rotation.Z {type=MovieSceneScriptingFloatChannel, default=0, const [0,321]}
       CTRL_L_eyeAim.Scale.X {type=MovieSceneScriptingFloatChannel, default=1, const [1,321]}
       CTRL_L_eyeAim.Scale.Y {type=MovieSceneScriptingFloatChannel, default=1, const [1,321]}
       CTRL_L_eyeAim.Scale.Z {type=MovieSceneScriptingFloatChannel, default=1, const [1,321]}
       CTRL_R_eyeAim.Location.X {
        type=MovieSceneScriptingFloatChannel
        default=0
        thinned_from=321
        keys [[0,0.3171],[1,0.3171],[2,0.3217],[3,0.3279],[4,0.3359],[5,0.3458],[8,0.3769],[9,0.3866],[12,0.3732],[17,0.3462],[18,0.321],[19,0.2811],[20,0.2591],[21,0.2422],[22,0.2238],[26,0.2038],[28,0.1947],[30,0.1869],[32,0.1802],[34,0.1744],[36,0.1694],[37,0.1672],[38,0.1566],[39,0.1519],[40,0.1495],[41,0.1482],[42,0.1474],[43,0.1471],[44,0.1472],[45,0.1476],[46,0.1485],[47,0.1497],[48,0.1624],[49,0.1676],[50,0.1703],[52,0.1751],[55,0.1833],[57,0.188],[60,0.1938],[61,0.1876],[62,0.1853],[63,0.1825],[64,0.1777],[65,0.1712],[66,0.1663],[67,0.1888],[68,0.177],[69,0.1571],[70,0.1342],[71,0.1095],[72,0.0803],[73,0.046],[74,0.0112],[75,-0.0422],[76,-0.096],[77,-0.1476],[78,-0.1949],[79,-0.2337],[80,-0.2645],[81,-0.2868],[82,-0.3035],[83,-0.3013],[84,-0.2937],[85,-0.2851],[87,-0.2656],[89,-0.2529],[90,-0.2527],[91,-0.2533],[92,-0.2546],[96,-0.2635],[97,-0.2651],[99,-0.2667],[104,-0.2664],[105,-0.2463],[106,-0.2384],[107,-0.235],[108,-0.233],[112,-0.2291],[113,-0.2273],[114,-0.2139],[115,-0.2118],[117,-0.2135],[118,-0.2139],[119,-0.2138],[120,-0.2131],[121,-0.2119],[122,-0.2098],[123,-0.207],[124,-0.2036],[125,-0.1995],[126,-0.1947],[127,-0.1891],[128,-0.1828],[129,-0.1755],[130,-0.1671],[131,-0.1574],[132,-0.1462],[133,-0.1771],[134,-0.177],[136,-0.1639],[138,-0.1496],[139,-0.1404],[140,-0.1277],[141,-0.1146],[142,-0.1159],[143,-0.0897],[144,-0.0564],[145,-0.0699],[146,-0.1269],[147,-0.1038],[148,-0.0783],[149,-0.0513],[150,-0.0248],[151,-0.0017],[152,0.0183],[153,0.0349],[154,0.0479],[155,0.057],[156,0.0629],[157,0.0657],[158,0.0664],[159,0.0659],[160,0.0646],[161,0.063],[163,0.0587],[165,0.0538],[166,0.0471],[167,0.0494],[168,0.0473],[169,0.0458],[170,0.0448],[171,0.0434],[172,0.041],[173,0.0374],[174,0.0153],[175,0.0128],[176,0.0188],[177,0.0256],[178,0.0361],[179,0.0493],[180,0.0619],[181,0.0802],[182,0.1658],[183,0.1621],[184,0.2459],[185,0.2753],[186,0.2832],[187,0.3095],[188,0.3339],[189,0.3554],[190,0.3734],[191,0.3802],[192,0.3886],[193,0.3956],[194,0.4013],[195,0.4054],[196,0.4085],[197,0.4103],[198,0.411],[201,0.4091],[203,0.4099],[204,0.4119],[205,0.416],[206,0.4216],[207,0.4286],[208,0.4368],[211,0.4663],[212,0.4751],[213,0.482],[214,0.4865],[215,0.4882],[216,0.4888],[217,0.4878],[222,0.4758],[224,0.4758],[225,0.4706],[227,0.4673],[228,0.4617],[230,0.4617],[231,0.4637],[232,0.4642],[233,0.4632],[234,0.4595],[235,0.4521],[236,0.44],[237,0.4227],[238,0.4002],[239,0.3945],[240,0.5894],[241,0.2918],[242,0.2516],[243,0.2337],[244,0.2055],[245,0.1724],[246,0.1545],[247,0.1428],[248,0.1328],[250,0.1172],[251,0.1143],[252,0.1166],[253,0.1166],[254,0.1149],[255,0.1119],[256,0.1085],[257,0.1042],[258,0.0988],[261,0.0817],[262,0.0764],[263,0.0719],[264,0.0682],[265,0.0657],[266,0.0644],[267,0.0643],[268,0.0651],[269,0.0916],[270,0.1006],[271,0.1052],[274,0.1176],[275,0.1221],[276,0.1269],[277,0.1321],[278,0.1376],[280,0.1494],[282,0.1617],[287,0.1914],[288,0.177],[289,0.1758],[290,0.1808],[291,0.1879],[295,0.22],[296,0.2275],[297,0.2325],[298,0.2392],[300,0.2539],[302,0.2672],[303,0.273],[304,0.2783],[305,0.2829],[306,0.2933],[307,0.2999],[309,0.3061],[310,0.3117],[311,0.3158],[312,0.3185],[313,0.3201],[315,0.3212],[318,0.3206],[319,0.3191],[320,0.3191]]
       }
       CTRL_R_eyeAim.Location.Y {
        type=MovieSceneScriptingFloatChannel
        default=0
        thinned_from=321
        keys [[0,1.0046],[15,0.9865],[17,0.9814],[18,0.8626],[19,0.8251],[20,0.8134],[28,0.7798],[37,0.7504],[38,0.7842],[39,0.7987],[40,0.8027],[43,0.7893],[47,0.7648],[48,0.8235],[49,0.8355],[50,0.8329],[51,0.8227],[52,0.8069],[55,0.7482],[56,0.7344],[58,0.7152],[59,0.7011],[60,0.6566],[61,0.4689],[62,0.4091],[63,0.3727],[64,0.3508],[65,0.3352],[66,0.2584],[67,-0.4862],[68,-0.9596],[69,-1.3],[70,-1.4761],[71,-1.5561],[72,-1.5619],[73,-1.5456],[75,-1.4616],[76,-1.4221],[77,-1.3694],[78,-1.3071],[79,-1.2372],[80,-1.162],[81,-1.0994],[82,-1.0509],[83,-1.0112],[84,-0.9792],[85,-0.9518],[86,-0.9304],[87,-0.9157],[88,-0.9061],[89,-0.7826],[90,-0.7563],[91,-0.7427],[92,-0.738],[93,-0.7376],[94,-0.7393],[96,-0.7472],[97,-0.7536],[98,-0.7617],[100,-0.7828],[102,-0.809],[104,-0.8384],[105,-0.9696],[106,-1.0222],[107,-1.052],[109,-1.0997],[110,-1.1206],[111,-1.1392],[112,-1.1554],[113,-1.1691],[114,-1.1175],[115,-1.1107],[116,-1.1175],[125,-1.2008],[127,-1.2233],[129,-1.252],[130,-1.2722],[131,-1.3004],[132,-1.3395],[133,-0.9113],[134,-0.87],[135,-0.8925],[136,-0.8299],[137,-0.7443],[138,-0.6828],[139,-0.6876],[140,-0.7486],[141,-0.8043],[142,-0.6983],[143,-0.903],[144,-1.3877],[145,-2.6894],[146,0.2023],[147,0.1704],[148,0.0395],[149,-0.1501],[150,-0.3416],[151,-0.4816],[152,-0.5945],[153,-0.6899],[154,-0.7667],[155,-0.8262],[156,-0.867],[157,-0.8905],[158,-0.9027],[160,-0.9197],[161,-0.9268],[162,-0.9292],[166,-0.9237],[167,-0.9257],[168,-0.9322],[169,-0.9522],[171,-1.0236],[172,-1.0497],[173,-1.0572],[174,-0.9372],[175,-0.9163],[176,-0.9156],[177,-0.9089],[178,-0.9177],[179,-0.9578],[180,-1.0743],[181,-1.1272],[182,-0.8893],[183,-0.4681],[184,0.9558],[185,0.8676],[186,0.7939],[188,0.6519],[189,0.584],[190,0.518],[191,0.4543],[192,0.3929],[193,0.334],[194,0.2798],[195,0.2322],[196,0.1918],[197,0.1589],[198,0.1312],[199,0.1053],[200,0.0822],[201,0.06],[202,0.0386],[203,0.0178],[204,0.0033],[215,0.0047],[230,0.0042],[231,0.0266],[232,0.074],[233,0.1336],[234,0.202],[235,0.2885],[236,0.4078],[237,0.566],[238,0.7549],[239,0.9657],[240,-0.9829],[241,-0.22],[242,-0.0665],[243,0.0219],[244,0.2522],[245,0.5435],[246,0.7323],[247,0.8182],[248,0.8873],[249,0.9347],[250,0.97],[251,0.9926],[252,0.9866],[253,1.0036],[254,1.0339],[255,1.0681],[256,1.0996],[257,1.1275],[258,1.1519],[259,1.1729],[260,1.1907],[261,1.2054],[262,1.2172],[263,1.2261],[264,1.2321],[265,1.235],[266,1.2348],[268,1.2246],[269,1.2087],[273,1.1764],[275,1.1546],[277,1.1275],[279,1.0955],[281,1.059],[283,1.0175],[285,0.9717],[287,0.9221],[288,1.0068],[289,1.0603],[290,1.065],[291,1.0554],[294,0.9906],[295,0.9703],[296,0.9763],[297,1.1248],[298,1.1531],[299,1.1497],[300,1.1301],[301,1.1063],[302,1.0765],[304,1.0075],[305,0.9783],[306,1.0897],[307,1.043],[308,1.0689],[309,1.0567],[310,1.0405],[313,0.9835],[314,0.9673],[315,0.9537],[316,0.9428],[317,0.9346],[318,0.9289],[319,0.9258],[320,0.9258]]
       }
       CTRL_R_eyeAim.Location.Z {
        type=MovieSceneScriptingFloatChannel
        default=0
        thinned_from=321
        keys [[0,0.2433],[6,0.2388],[17,0.2239],[18,0.2028],[19,0.1961],[20,0.1936],[25,0.1871],[31,0.1821],[37,0.1788],[38,0.1836],[39,0.1858],[40,0.1865],[47,0.184],[48,0.1933],[49,0.1958],[50,0.1962],[51,0.1955],[52,0.194],[55,0.1882],[56,0.1871],[58,0.1862],[59,0.185],[60,0.1801],[61,0.1613],[62,0.157],[63,0.1549],[64,0.154],[65,0.1536],[66,0.1502],[67,0.1667],[68,0.2308],[69,0.3037],[70,0.351],[71,0.375],[72,0.377],[73,0.3718],[74,0.3587],[76,0.33],[77,0.3112],[78,0.2894],[80,0.242],[81,0.2232],[82,0.2088],[83,0.1969],[84,0.1873],[85,0.1791],[86,0.1725],[87,0.1678],[88,0.1646],[89,0.1443],[90,0.1398],[91,0.1373],[92,0.1361],[93,0.1356],[94,0.1357],[95,0.1361],[96,0.137],[97,0.1383],[98,0.1401],[99,0.1425],[100,0.1453],[101,0.1485],[103,0.1557],[104,0.1596],[105,0.1832],[106,0.1946],[107,0.2019],[109,0.2147],[110,0.2206],[111,0.2259],[112,0.2308],[113,0.2351],[114,0.2253],[115,0.2249],[116,0.2273],[124,0.2487],[126,0.2552],[127,0.259],[129,0.2684],[130,0.2749],[131,0.2842],[132,0.297],[133,0.2106],[134,0.2081],[135,0.2162],[136,0.21],[137,0.2005],[138,0.1942],[139,0.1955],[140,0.2038],[141,0.2116],[142,0.1962],[143,0.2254],[144,0.3273],[145,0.8489],[146,0.1423],[147,0.1298],[148,0.1103],[149,0.0895],[150,0.0725],[151,0.0616],[152,0.0539],[153,0.048],[154,0.0435],[155,0.0398],[156,0.0359],[157,0.0312],[159,0.021],[160,0.0167],[161,0.0127],[164,0],[165,-0.0038],[166,-0.0073],[167,-0.0095],[168,-0.011],[169,-0.01],[170,-0.0061],[171,-0.0014],[172,0.0015],[173,0.0006],[174,-0.0247],[175,-0.0322],[176,-0.0363],[177,-0.0413],[178,-0.0449],[179,-0.0462],[180,-0.0374],[181,-0.0469],[182,-0.1184],[183,-0.2023],[184,-0.1708],[185,-0.2142],[186,-0.2506],[187,-0.2849],[188,-0.3181],[189,-0.3496],[190,-0.3798],[191,-0.409],[192,-0.4366],[193,-0.4625],[194,-0.4861],[195,-0.5066],[196,-0.5237],[197,-0.5375],[198,-0.5344],[199,-0.4303],[200,-0.3359],[201,-0.2434],[202,-0.1527],[203,-0.0631],[204,0],[230,0],[231,-0.0972],[232,-0.2979],[233,-0.5403],[234,-0.5194],[235,-0.4826],[236,-0.4306],[237,-0.3597],[238,-0.2721],[239,-0.1703],[240,-0.0804],[241,-0.118],[242,-0.0663],[243,-0.0184],[244,0.0265],[245,0.0776],[246,0.1197],[247,0.1442],[248,0.163],[249,0.1772],[250,0.1895],[251,0.1991],[252,0.2023],[253,0.2087],[255,0.2253],[256,0.233],[257,0.2398],[258,0.2458],[259,0.251],[260,0.2555],[261,0.2592],[262,0.2621],[263,0.2643],[264,0.2658],[265,0.2666],[266,0.2666],[268,0.2642],[269,0.2605],[273,0.2542],[275,0.25],[278,0.2418],[282,0.2284],[287,0.2097],[288,0.2256],[289,0.2367],[290,0.2385],[291,0.2376],[294,0.2285],[295,0.226],[296,0.2281],[297,0.2581],[298,0.265],[299,0.2652],[300,0.2618],[301,0.2575],[302,0.2522],[304,0.2401],[305,0.2352],[306,0.257],[307,0.2482],[308,0.2537],[309,0.2517],[313,0.239],[314,0.2363],[316,0.2322],[318,0.2299],[319,0.2293],[320,0.2293]]
       }
       CTRL_R_eyeAim.Rotation.X {
        type=MovieSceneScriptingFloatChannel
        default=0
        thinned_from=321
        keys [[0,0.3546],[4,0.3498],[10,0.3375],[18,0.3169],[23,0.3078],[28,0.3021],[36,0.299],[41,0.2995],[44,0.3019],[47,0.3061],[59,0.3291],[65,0.3367],[67,0.3406],[70,0.3495],[71,0.352],[72,0.3529],[73,0.3519],[74,0.349],[75,0.3435],[76,0.3351],[77,0.3233],[78,0.3087],[80,0.2756],[81,0.2622],[82,0.2511],[83,0.2412],[84,0.2326],[85,0.2246],[86,0.218],[87,0.213],[88,0.2093],[89,0.2067],[91,0.2033],[93,0.2012],[94,0.2007],[95,0.2007],[96,0.2011],[97,0.2022],[98,0.2038],[99,0.2061],[100,0.2089],[103,0.2193],[110,0.2454],[113,0.2551],[116,0.2622],[124,0.2754],[127,0.2822],[129,0.2885],[130,0.2926],[131,0.2981],[132,0.3052],[133,0.3145],[135,0.335],[136,0.344],[137,0.3508],[138,0.3548],[139,0.3564],[140,0.3568],[141,0.3562],[142,0.3547],[143,0.352],[144,0.3473],[145,0.3398],[146,0.3254],[147,0.2985],[148,0.2586],[149,0.2052],[150,0.1453],[151,0.095],[152,0.0507],[153,0.0105],[154,-0.0242],[155,-0.0534],[156,-0.0775],[157,-0.0975],[158,-0.1144],[159,-0.1297],[160,-0.1433],[161,-0.1557],[162,-0.1667],[163,-0.1765],[164,-0.1854],[166,-0.2013],[169,-0.2193],[170,-0.2248],[172,-0.2347],[173,-0.2402],[174,-0.2472],[176,-0.2659],[177,-0.2749],[178,-0.287],[179,-0.3064],[180,-0.3374],[181,-0.3855],[182,-0.4493],[183,-0.5223],[184,-0.5968],[185,-0.6637],[186,-0.7224],[189,-0.8925],[191,-1.0032],[193,-1.1087],[194,-1.157],[195,-1.2],[196,-1.2367],[197,-1.2667],[198,-1.2578],[199,-1.0123],[200,-0.7898],[201,-0.5722],[202,-0.3587],[203,-0.1483],[204,0],[230,0],[231,-0.2283],[232,-0.7003],[233,-1.2718],[234,-1.2274],[235,-1.1497],[236,-1.0451],[237,-0.9113],[238,-0.7591],[239,-0.5988],[240,-0.4203],[241,-0.2878],[242,-0.1567],[243,-0.0433],[244,0.0486],[245,0.1185],[246,0.1653],[247,0.1937],[248,0.2121],[249,0.2265],[250,0.2402],[251,0.2532],[252,0.2633],[253,0.271],[254,0.2767],[255,0.2806],[257,0.2859],[259,0.2892],[262,0.2919],[268,0.293],[284,0.3057],[288,0.3114],[299,0.3375],[306,0.348],[315,0.3545],[320,0.3548]]
       }
       CTRL_R_eyeAim.Rotation.Y {
        type=MovieSceneScriptingFloatChannel
        default=0
        thinned_from=321
        keys [[0,0],[1,-0.0002],[4,-0.0018],[11,-0.0066],[13,-0.0076],[15,-0.0081],[17,-0.0082],[18,-0.0039],[19,-0.0022],[20,-0.0013],[24,0.0017],[28,0.0052],[33,0.0085],[37,0.0104],[38,0.0135],[39,0.0146],[47,0.0172],[48,0.0116],[49,0.01],[50,0.0097],[55,0.0091],[60,0.0074],[61,0.0146],[62,0.0172],[63,0.0179],[64,0.0177],[66,0.0165],[67,0.0115],[68,0.0056],[69,0.0028],[70,0.0016],[76,-0.0013],[78,-0.0019],[79,-0.0018],[81,-0.0008],[84,0.0009],[88,0.0023],[89,0.0046],[90,0.0053],[91,0.0058],[93,0.0061],[104,0.0064],[105,0.0105],[106,0.0119],[107,0.0125],[112,0.0134],[113,0.0138],[114,0.0176],[115,0.0183],[119,0.0177],[122,0.0179],[125,0.0189],[127,0.02],[129,0.0216],[131,0.0237],[132,0.0251],[133,0.0241],[134,0.0239],[135,0.0245],[136,0.0271],[137,0.0312],[138,0.0351],[139,0.0371],[140,0.0375],[141,0.0377],[142,0.0377],[143,0.0354],[144,0.0284],[145,0.0114],[146,0.0045],[149,0.003],[153,0.0002],[160,-0.0064],[163,-0.0086],[166,-0.0098],[169,-0.01],[171,-0.0095],[172,-0.0089],[173,-0.0079],[174,-0.0029],[176,0.0003],[177,0.0026],[178,0.0052],[179,0.008],[180,0.0114],[181,0.0154],[182,-0.0251],[183,0.0247],[184,0.0266],[185,0.0339],[186,0.0245],[187,0.0314],[188,0.0393],[189,0.0482],[190,0.0576],[191,0.052],[192,0.0542],[193,0.0597],[194,0.0673],[195,0.0744],[196,0.0807],[197,0.0861],[198,0.0881],[199,0.0724],[202,0.0267],[203,0.0112],[204,0],[230,0],[231,0.0159],[232,0.0466],[233,0.0785],[234,0.0675],[235,0.0537],[236,0.0387],[237,0.0245],[238,0.0131],[239,0.0187],[240,-0.1315],[241,-0.0432],[242,-0.0142],[243,-0.0033],[244,0.0039],[245,0.0099],[246,0.0132],[247,0.0161],[249,0.0209],[250,0.0235],[251,0.0251],[252,0.0259],[257,0.0273],[262,0.0276],[265,0.0271],[268,0.0259],[269,0.0195],[270,0.0172],[271,0.0161],[284,0.0044],[286,0.0029],[287,0.0024],[288,0.0077],[289,0.009],[290,0.009],[296,0.0078],[297,0.0068],[305,0.005],[306,0.0016],[307,0.0005],[310,-0.0002],[320,-0.0001]]
       }
       CTRL_R_eyeAim.Rotation.Z {
        type=MovieSceneScriptingFloatChannel
        default=0
        thinned_from=321
        keys [[0,0.7766],[1,0.7751],[2,0.7834],[3,0.7948],[4,0.8101],[8,0.8892],[9,0.9078],[16,0.7998],[17,0.7856],[18,0.7555],[19,0.6692],[20,0.621],[21,0.5838],[22,0.544],[25,0.5227],[28,0.5049],[31,0.4903],[36,0.4715],[39,0.4633],[41,0.46],[43,0.4595],[45,0.4619],[48,0.4699],[55,0.4953],[57,0.501],[59,0.5039],[60,0.5028],[61,0.4985],[62,0.4928],[63,0.4828],[64,0.4673],[65,0.4482],[66,0.4246],[67,0.3968],[68,0.3642],[69,0.3264],[70,0.2822],[71,0.2306],[72,0.1704],[73,0.0994],[74,0.0279],[75,-0.0854],[76,-0.2008],[77,-0.3129],[78,-0.4169],[79,-0.5057],[80,-0.58],[81,-0.6373],[82,-0.6825],[83,-0.6834],[84,-0.6718],[85,-0.6571],[87,-0.6208],[88,-0.6096],[96,-0.6523],[98,-0.6586],[100,-0.6611],[114,-0.6517],[118,-0.651],[121,-0.6461],[123,-0.6395],[125,-0.6297],[127,-0.6169],[129,-0.6012],[131,-0.5819],[133,-0.5589],[139,-0.484],[140,-0.4703],[141,-0.454],[142,-0.434],[143,-0.4087],[144,-0.3775],[145,-0.3391],[146,-0.2939],[147,-0.2401],[148,-0.1828],[149,-0.1249],[150,-0.0703],[151,-0.0243],[152,0.014],[153,0.0436],[154,0.0645],[155,0.0757],[156,0.0791],[157,0.0763],[158,0.0697],[159,0.062],[160,0.0534],[161,0.0454],[162,0.0379],[163,0.0311],[164,0.025],[165,0.0197],[166,0.0047],[167,0.0114],[168,0.0084],[169,0.006],[170,0.0043],[171,0.0031],[172,0.0034],[173,0.0058],[174,0.0102],[175,0.0183],[176,0.0449],[177,0.0774],[178,0.1192],[179,0.1678],[180,0.2206],[181,0.281],[182,0.2632],[183,0.4243],[184,0.4967],[185,0.5622],[186,0.6216],[187,0.6783],[188,0.7313],[189,0.7789],[190,0.8199],[191,0.8531],[192,0.879],[193,0.8987],[194,0.914],[195,0.9265],[196,0.9365],[197,0.9435],[199,0.9507],[201,0.9527],[203,0.9605],[204,0.967],[205,0.9767],[206,0.9898],[207,1.0062],[208,1.0256],[211,1.0949],[212,1.1155],[213,1.1316],[214,1.1422],[215,1.1464],[216,1.1476],[217,1.1453],[222,1.1172],[224,1.1172],[225,1.1049],[227,1.0972],[228,1.0841],[230,1.0841],[231,1.086],[232,1.0818],[233,1.0737],[234,1.0612],[235,1.042],[236,1.0144],[237,0.9793],[238,0.9378],[239,0.8907],[240,0.6332],[241,0.6018],[242,0.5747],[243,0.5528],[244,0.5358],[245,0.5231],[246,0.5156],[247,0.5129],[248,0.5134],[252,0.5231],[254,0.5259],[256,0.5256],[257,0.5218],[258,0.5146],[259,0.5057],[263,0.4631],[264,0.4534],[265,0.4452],[266,0.4383],[267,0.4327],[269,0.4247],[270,0.4219],[272,0.4191],[274,0.4194],[276,0.4228],[278,0.4289],[280,0.4375],[283,0.4538],[284,0.4601],[286,0.4754],[287,0.4844],[288,0.4947],[289,0.5062],[290,0.5189],[292,0.5478],[298,0.6445],[300,0.6735],[302,0.6983],[304,0.7182],[306,0.7333],[309,0.7511],[310,0.763],[311,0.7717],[312,0.7775],[313,0.7809],[315,0.7832],[318,0.7817],[319,0.7783],[320,0.7783]]
       }
       CTRL_R_eyeAim.Scale.X {type=MovieSceneScriptingFloatChannel, default=1, const [1,321]}
       CTRL_R_eyeAim.Scale.Y {type=MovieSceneScriptingFloatChannel, default=1, const [1,321]}
       CTRL_R_eyeAim.Scale.Z {type=MovieSceneScriptingFloatChannel, default=1, const [1,321]}
       CTRL_convergenceSwitch {type=MovieSceneScriptingFloatChannel, default=0, const [0,321]}
       CTRL_Master.Location.X {type=MovieSceneScriptingFloatChannel, default=0, const [0,321]}
       CTRL_Master.Location.Y {type=MovieSceneScriptingFloatChannel, default=0, const [0,321]}
       CTRL_Master.Location.Z {type=MovieSceneScriptingFloatChannel, default=0, const [0,321]}
       CTRL_Master.Rotation.X {type=MovieSceneScriptingFloatChannel, default=0, const [0,321]}
       CTRL_Master.Rotation.Y {type=MovieSceneScriptingFloatChannel, default=0, const [0,321]}
       CTRL_Master.Rotation.Z {type=MovieSceneScriptingFloatChannel, default=0, const [0,321]}
       CTRL_Master.Scale.X {type=MovieSceneScriptingFloatChannel, default=1, const [1,321]}
       CTRL_Master.Scale.Y {type=MovieSceneScriptingFloatChannel, default=1, const [1,321]}
       CTRL_Master.Scale.Z {type=MovieSceneScriptingFloatChannel, default=1, const [1,321]}
       CTRL_C_jaw.X {type=MovieSceneScriptingFloatChannel, default=0, const [0,321]}
       CTRL_C_jaw.Y {type=MovieSceneScriptingFloatChannel, default=0, const [0,321]}
       CTRL_C_jaw_openExtreme {type=MovieSceneScriptingFloatChannel, default=0, const [0,321]}
       CTRL_C_jaw_fwdBack {type=MovieSceneScriptingFloatChannel, default=0, const [0,321]}
       CTRL_R_jaw_ChinRaiseD {
        type=MovieSceneScriptingFloatChannel
        default=0
        thinned_from=321
        keys [[0,0.0927],[4,0.0927],[6,0.0925],[10,0.0916],[12,0.0907],[21,0.0839],[26,0.0788],[28,0.0772],[29,0.0766],[33,0.0758],[46,0.0756],[50,0.0762],[52,0.0769],[56,0.0795],[61,0.0835],[67,0.0897],[69,0.0912],[72,0.0921],[74,0.0924],[85,0.0926],[91,0.0922],[95,0.0916],[100,0.0913],[104,0.0904],[107,0.0889],[118,0.081],[129,0.0755],[138,0.073],[146,0.0719],[161,0.0715],[177,0.0723],[189,0.0746],[195,0.077],[205,0.0833],[215,0.087],[240,0.0909],[270,0.101],[288,0.1043],[297,0.1043],[300,0.1039],[301,0.1035],[302,0.1029],[304,0.1009],[308,0.096],[309,0.0951],[310,0.0944],[313,0.0933],[316,0.0928],[320,0.0927]]
       }
       CTRL_R_jaw_ChinRaiseU {type=MovieSceneScriptingFloatChannel, default=0, const [0,321]}
       CTRL_L_jaw_ChinRaiseU {type=MovieSceneScriptingFloatChannel, default=0, const [0,321]}
       CTRL_L_jaw_ChinRaiseD {
        type=MovieSceneScriptingFloatChannel
        default=0
        thinned_from=321
        keys [[0,0.0927],[4,0.0927],[6,0.0925],[10,0.0916],[12,0.0907],[21,0.0839],[26,0.0788],[28,0.0772],[29,0.0766],[33,0.0758],[46,0.0756],[50,0.0762],[52,0.0769],[56,0.0795],[61,0.0835],[67,0.0897],[69,0.0912],[72,0.0921],[74,0.0924],[85,0.0926],[91,0.0922],[95,0.0916],[100,0.0913],[104,0.0904],[107,0.0889],[118,0.081],[129,0.0755],[138,0.073],[146,0.0719],[161,0.0715],[177,0.0723],[189,0.0746],[195,0.077],[205,0.0833],[215,0.087],[240,0.0909],[270,0.101],[288,0.1043],[297,0.1043],[300,0.1039],[301,0.1035],[302,0.1029],[304,0.1009],[308,0.096],[309,0.0951],[310,0.0944],[313,0.0933],[316,0.0928],[320,0.0927]]
       }
       CTRL_R_mouth_lipsBlow {type=MovieSceneScriptingFloatChannel, default=0, const [0,321]}
       CTRL_L_mouth_lipsTogetherU {type=MovieSceneScriptingFloatChannel, default=0, const [0,321]}
       CTRL_L_mouth_lipsTogetherD {type=MovieSceneScriptingFloatChannel, default=0, const [0,321]}
       CTRL_R_mouth_lipsTogetherU {type=MovieSceneScriptingFloatChannel, default=0, const [0,321]}
       CTRL_L_mouth_lipsBlow {type=MovieSceneScriptingFloatChannel, default=0, const [0,321]}
       CTRL_R_mouth_lipsTogetherD {type=MovieSceneScriptingFloatChannel, default=0, const [0,321]}
       CTRL_C_tongue_roll.X {type=MovieSceneScriptingFloatChannel, default=0, const [0,321]}
       CTRL_C_tongue_roll.Y {type=MovieSceneScriptingFloatChannel, default=0, const [0,321]}
       CTRL_C_tongue.X {type=MovieSceneScriptingFloatChannel, default=0, const [0,321]}
       CTRL_C_tongue.Y {type=MovieSceneScriptingFloatChannel, default=0, const [0,321]}
       CTRL_C_tongue_narrowWide {type=MovieSceneScriptingFloatChannel, default=0, const [0,321]}
       CTRL_C_tongue_inOut {type=MovieSceneScriptingFloatChannel, default=0, const [0,321]}
       CTRL_R_nose.X {type=MovieSceneScriptingFloatChannel, default=0, const [0,321]}
       CTRL_R_nose.Y {type=MovieSceneScriptingFloatChannel, default=0, const [0,321]}
       CTRL_L_nose.X {type=MovieSceneScriptingFloatChannel, default=0, const [0,321]}
       CTRL_L_nose.Y {type=MovieSceneScriptingFloatChannel, default=0, const [0,321]}
       CTRL_L_eye.X {
        type=MovieSceneScriptingFloatChannel
        default=0
        thinned_from=321
        keys [[0,0.2628],[16,0.2574],[17,0.2567],[18,0.2253],[19,0.2155],[20,0.2125],[28,0.2038],[37,0.1963],[38,0.2053],[39,0.2091],[40,0.2102],[43,0.2067],[47,0.2003],[48,0.2155],[49,0.2187],[50,0.218],[51,0.2153],[52,0.2111],[55,0.1956],[56,0.1919],[58,0.1869],[59,0.1831],[60,0.1714],[61,0.1224],[62,0.1067],[63,0.0972],[64,0.0915],[65,0.0874],[66,0.0673],[67,-0.1272],[68,-0.2517],[69,-0.3423],[70,-0.3898],[71,-0.4114],[72,-0.413],[73,-0.4085],[75,-0.3857],[76,-0.3751],[77,-0.361],[78,-0.3443],[79,-0.3257],[80,-0.3056],[81,-0.289],[82,-0.2762],[83,-0.2656],[84,-0.2571],[85,-0.2498],[86,-0.2442],[87,-0.2402],[88,-0.2377],[89,-0.2051],[90,-0.1982],[91,-0.1946],[92,-0.1934],[93,-0.1933],[94,-0.1938],[96,-0.1959],[97,-0.1976],[98,-0.1997],[100,-0.2053],[102,-0.2122],[104,-0.2199],[105,-0.2547],[106,-0.2687],[107,-0.2767],[109,-0.2894],[110,-0.2949],[111,-0.2999],[112,-0.3042],[113,-0.3079],[114,-0.2943],[115,-0.2925],[116,-0.2943],[125,-0.3166],[127,-0.3226],[129,-0.3304],[130,-0.3358],[131,-0.3435],[132,-0.3541],[133,-0.2396],[134,-0.2286],[135,-0.2345],[136,-0.218],[137,-0.1956],[138,-0.1795],[139,-0.1808],[140,-0.1969],[141,-0.2117],[142,-0.1837],[143,-0.2377],[144,-0.367],[145,-0.7332],[146,0.0527],[147,0.0444],[148,0.0103],[149,-0.0391],[150,-0.0891],[151,-0.1257],[152,-0.1553],[153,-0.1804],[154,-0.2006],[155,-0.2163],[156,-0.2272],[157,-0.2334],[158,-0.2367],[160,-0.2412],[161,-0.2431],[162,-0.2438],[166,-0.2424],[167,-0.2429],[168,-0.2446],[169,-0.2499],[171,-0.2688],[172,-0.2757],[173,-0.2776],[174,-0.2457],[175,-0.2402],[176,-0.24],[177,-0.2382],[178,-0.2406],[179,-0.2513],[180,-0.2823],[181,-0.2965],[182,-0.2337],[183,-0.1229],[184,0.2499],[185,0.2264],[186,0.2067],[188,0.1691],[189,0.1511],[190,0.1336],[191,0.1167],[192,0.1005],[193,0.0849],[194,0.0706],[195,0.0581],[196,0.0474],[197,0.0387],[198,0.0315],[199,0.0254],[201,0.0143],[203,0.0037],[204,0],[230,0],[231,0.0057],[232,0.0175],[233,0.0319],[234,0.0498],[235,0.0726],[236,0.1041],[237,0.1459],[238,0.1958],[239,0.2519],[240,-0.2718],[241,-0.0586],[242,-0.0178],[243,0.0054],[244,0.0658],[245,0.1423],[246,0.192],[247,0.2148],[248,0.2332],[249,0.2459],[250,0.2555],[251,0.2615],[252,0.2599],[253,0.2644],[254,0.2725],[255,0.2816],[256,0.29],[257,0.2975],[258,0.304],[259,0.3096],[260,0.3144],[261,0.3184],[262,0.3215],[263,0.3239],[264,0.3255],[265,0.3263],[266,0.3263],[268,0.3233],[269,0.3184],[273,0.3094],[275,0.3035],[277,0.2962],[279,0.2876],[281,0.2778],[283,0.2667],[285,0.2546],[287,0.2414],[288,0.2639],[289,0.2781],[290,0.2794],[291,0.2768],[294,0.2595],[295,0.2541],[296,0.2557],[297,0.2951],[298,0.3026],[299,0.3017],[300,0.2965],[301,0.2901],[302,0.2821],[304,0.2638],[305,0.256],[306,0.2855],[307,0.2731],[308,0.28],[309,0.2767],[310,0.2724],[313,0.2573],[314,0.253],[315,0.2493],[316,0.2465],[317,0.2443],[318,0.2428],[319,0.242],[320,0.2418]]
       }
       CTRL_L_eye.Y {
        type=MovieSceneScriptingFloatChannel
        default=0
        thinned_from=321
        keys [[0,-0.0195],[1,-0.0187],[2,-0.0171],[3,-0.0152],[4,-0.0131],[9,-0.0001],[11,0.0063],[12,0.0092],[13,0.0117],[14,0.0138],[15,0.0154],[16,0.0164],[17,0.017],[18,-0.0012],[19,-0.0068],[20,-0.0095],[21,-0.0114],[22,-0.014],[23,-0.0174],[24,-0.0212],[27,-0.0336],[28,-0.0375],[29,-0.041],[31,-0.0473],[33,-0.0528],[35,-0.0575],[37,-0.0617],[38,-0.0762],[39,-0.0815],[40,-0.0839],[42,-0.0869],[44,-0.0894],[46,-0.0913],[47,-0.092],[48,-0.0657],[49,-0.0582],[50,-0.0565],[52,-0.0555],[54,-0.054],[55,-0.053],[57,-0.0502],[59,-0.0466],[60,-0.0446],[61,-0.0757],[62,-0.0863],[63,-0.089],[64,-0.0871],[65,-0.084],[66,-0.0805],[67,-0.0582],[68,-0.0327],[69,-0.0199],[70,-0.0135],[72,-0.0071],[73,-0.0033],[74,0.0007],[76,0.0105],[77,0.015],[78,0.019],[79,0.0212],[80,0.0215],[81,0.02],[82,0.0174],[83,0.0144],[85,0.0079],[86,0.0048],[87,0.0019],[88,-0.0004],[89,-0.0169],[90,-0.0219],[91,-0.0254],[92,-0.0272],[93,-0.0278],[100,-0.0265],[102,-0.0258],[104,-0.0245],[105,-0.0503],[106,-0.0579],[107,-0.0602],[108,-0.0609],[109,-0.0607],[110,-0.0601],[111,-0.0597],[112,-0.0601],[113,-0.0615],[114,-0.0813],[115,-0.0846],[116,-0.0832],[117,-0.0812],[118,-0.0796],[119,-0.0786],[120,-0.078],[121,-0.0778],[122,-0.0782],[123,-0.079],[124,-0.0802],[125,-0.0818],[126,-0.0837],[127,-0.086],[128,-0.0887],[129,-0.0919],[130,-0.0955],[132,-0.1035],[133,-0.0955],[134,-0.0916],[135,-0.0913],[136,-0.0999],[137,-0.1146],[138,-0.129],[139,-0.1365],[140,-0.1383],[141,-0.1397],[142,-0.1408],[143,-0.1334],[144,-0.1075],[145,-0.0397],[146,-0.0125],[147,-0.0134],[148,-0.0153],[149,-0.0182],[150,-0.0216],[151,-0.0256],[153,-0.0338],[154,-0.0381],[155,-0.043],[157,-0.0534],[158,-0.0581],[159,-0.062],[160,-0.0651],[161,-0.0674],[162,-0.0689],[163,-0.07],[164,-0.0704],[165,-0.0703],[166,-0.0699],[167,-0.0689],[168,-0.0671],[169,-0.0649],[170,-0.0623],[171,-0.0589],[172,-0.054],[173,-0.0473],[174,-0.0167],[175,-0.0078],[176,0.0006],[177,0.0116],[178,0.0229],[179,0.0332],[180,0.0428],[181,0.0502],[182,-0.0864],[183,0.057],[184,0.0514],[185,0.0591],[186,0.0332],[187,0.0408],[188,0.0491],[190,0.0665],[191,0.053],[192,0.0514],[193,0.0546],[194,0.0603],[195,0.0655],[196,0.07],[197,0.0736],[198,0.0765],[199,0.0786],[200,0.08],[201,0.0811],[202,0.0825],[203,0.0842],[204,0.0862],[205,0.0885],[206,0.0911],[207,0.0941],[208,0.0713],[209,0.0674],[210,0.0694],[215,0.087],[217,0.0946],[218,0.1225],[219,0.133],[220,0.1374],[221,0.0724],[222,0.074],[223,0.0742],[224,0.0742],[228,0.0731],[229,0.0731],[230,0.0734],[231,0.0723],[232,0.0683],[233,0.0614],[234,0.0522],[235,0.0408],[236,0.0277],[237,0.0141],[238,0.0013],[239,0.0224],[240,-0.4501],[241,-0.2283],[242,-0.1436],[243,-0.1217],[244,-0.1293],[245,-0.1327],[246,-0.1272],[247,-0.1322],[249,-0.1447],[250,-0.1526],[251,-0.1549],[252,-0.1538],[254,-0.1507],[255,-0.15],[259,-0.1488],[262,-0.147],[264,-0.1448],[265,-0.1433],[266,-0.1415],[267,-0.1393],[268,-0.1369],[269,-0.1057],[270,-0.0943],[271,-0.0884],[272,-0.0835],[273,-0.0791],[277,-0.0624],[282,-0.0402],[283,-0.0361],[284,-0.0323],[285,-0.0288],[286,-0.0257],[287,-0.0232],[288,-0.0479],[289,-0.0538],[290,-0.0541],[293,-0.0512],[295,-0.0501],[296,-0.049],[297,-0.045],[300,-0.0419],[302,-0.0403],[305,-0.0388],[306,-0.025],[307,-0.0208],[308,-0.0194],[309,-0.0187],[310,-0.0183],[311,-0.0181],[313,-0.0181],[320,-0.0195]]
       }
       CTRL_L_eye_blink {
        type=MovieSceneScriptingFloatChannel
        default=0
        thinned_from=272
        keys [[0,0.1532],[5,0.1525],[9,0.1506],[13,0.147],[17,0.142],[20,0.1372],[21,0.1354],[23,0.1309],[27,0.1211],[28,0.119],[29,0.1172],[30,0.1158],[32,0.1138],[34,0.1126],[36,0.112],[44,0.1128],[45,0.1138],[46,0.115],[47,0.1165],[48,0.1184],[49,0.1206],[50,0.1234],[51,0.1272],[52,0.1317],[53,0.1366],[54,0.1411],[55,0.1447],[56,0.1475],[57,0.1499],[58,0.1517],[59,0.1528],[60,0.1532],[63,0.1532],[64,0.1794],[65,0.2911],[66,0.564],[67,1],[68,1],[69,0.6138],[70,0.4114],[71,0.3121],[72,0.2528],[73,0.2043],[74,0.1735],[75,0.1583],[76,0.1532],[106,0.1532],[126,0.157],[128,0.1552],[131,0.1537],[139,0.1532],[140,0.1614],[141,0.2049],[142,0.3183],[143,0.5679],[144,1],[145,1],[146,0.6175],[147,0.3883],[148,0.2806],[149,0.2245],[150,0.1866],[151,0.167],[152,0.1558],[153,0.1534],[176,0.1532],[180,0.1514],[181,0.1764],[182,0.3109],[183,0.5553],[184,0.5704],[185,0.0803],[186,0.0458],[187,0.0493],[188,0.0628],[190,0.0997],[191,0.1128],[192,0.1156],[193,0.1118],[195,0.105],[196,0.1019],[197,0.099],[198,0.0964],[199,0.0942],[200,0.0924],[202,0.0897],[204,0.0877],[206,0.0864],[208,0.0856],[215,0.0854],[220,0.0862],[221,0.0868],[222,0.0881],[223,0.09],[224,0.0923],[226,0.0976],[227,0.1011],[230,0.1132],[231,0.1165],[232,0.1189],[233,0.1199],[234,0.1199],[235,0.1202],[236,0.1661],[237,0.3151],[238,0.5641],[239,1],[240,1],[241,0.6186],[242,0.386],[243,0.2668],[244,0.2022],[245,0.1646],[246,0.1444],[247,0.1354],[254,0.1358],[256,0.1365],[258,0.1439],[259,0.1473],[260,0.1498],[261,0.1515],[262,0.1525],[263,0.153],[320,0.1532]]
       }
       CTRL_R_eye.X {
        type=MovieSceneScriptingFloatChannel
        default=0
        thinned_from=321
        keys [[0,0.2717],[16,0.2655],[17,0.2647],[18,0.2332],[19,0.2234],[20,0.2203],[28,0.2114],[37,0.2037],[38,0.2127],[39,0.2166],[40,0.2177],[44,0.2128],[47,0.2079],[48,0.2232],[49,0.2264],[50,0.2257],[51,0.2231],[52,0.219],[55,0.2036],[56,0.2],[58,0.1951],[59,0.1914],[60,0.1797],[61,0.1306],[62,0.1151],[63,0.1056],[64,0.0999],[65,0.0958],[66,0.0758],[67,-0.1187],[68,-0.2431],[69,-0.3337],[70,-0.381],[71,-0.4026],[72,-0.4041],[73,-0.3997],[75,-0.3771],[76,-0.3667],[77,-0.3529],[78,-0.3366],[79,-0.3183],[80,-0.2988],[81,-0.2825],[82,-0.2699],[83,-0.2596],[84,-0.2513],[85,-0.2442],[86,-0.2387],[87,-0.2349],[88,-0.2325],[89,-0.2],[90,-0.1931],[91,-0.1896],[92,-0.1884],[93,-0.1884],[94,-0.1888],[96,-0.1908],[97,-0.1925],[98,-0.1946],[100,-0.2],[102,-0.2068],[104,-0.2143],[105,-0.249],[106,-0.2629],[107,-0.2708],[109,-0.2833],[110,-0.2888],[111,-0.2936],[112,-0.2979],[113,-0.3015],[114,-0.2879],[115,-0.286],[116,-0.2878],[125,-0.3096],[127,-0.3155],[129,-0.3232],[130,-0.3285],[131,-0.336],[132,-0.3464],[133,-0.2317],[134,-0.2204],[135,-0.2261],[136,-0.2094],[137,-0.1868],[138,-0.1706],[139,-0.1719],[140,-0.188],[141,-0.2027],[142,-0.1747],[143,-0.2289],[144,-0.3583],[145,-0.7247],[146,0.0608],[147,0.0518],[148,0.0167],[149,-0.034],[150,-0.0855],[151,-0.1233],[152,-0.1541],[153,-0.1801],[154,-0.2012],[155,-0.2177],[156,-0.2291],[157,-0.2358],[158,-0.2395],[160,-0.2448],[161,-0.247],[162,-0.248],[166,-0.2474],[167,-0.2481],[168,-0.2499],[169,-0.2554],[171,-0.2745],[172,-0.2816],[173,-0.2837],[174,-0.2519],[175,-0.2466],[176,-0.2466],[177,-0.2451],[178,-0.2478],[179,-0.2589],[180,-0.2907],[181,-0.3062],[182,-0.2449],[183,-0.1359],[184,0.235],[185,0.2098],[186,0.1886],[187,0.1682],[188,0.1481],[189,0.1287],[190,0.1099],[191,0.0916],[192,0.074],[193,0.0572],[194,0.0417],[195,0.028],[196,0.0164],[197,0.007],[198,0],[233,0],[234,0.0191],[235,0.0439],[236,0.078],[237,0.1231],[238,0.1768],[239,0.2369],[240,-0.2828],[241,-0.0659],[242,-0.0217],[243,0.0044],[244,0.067],[245,0.1453],[246,0.1962],[247,0.2197],[248,0.2386],[249,0.2516],[250,0.2615],[251,0.2679],[252,0.2665],[253,0.2712],[254,0.2794],[255,0.2886],[256,0.2971],[257,0.3046],[258,0.3112],[259,0.3169],[260,0.3217],[261,0.3257],[262,0.3288],[263,0.3312],[264,0.3328],[265,0.3336],[266,0.3336],[268,0.3307],[269,0.3258],[273,0.3169],[275,0.311],[277,0.3037],[279,0.2951],[281,0.2854],[283,0.2743],[286,0.2558],[287,0.2491],[288,0.2717],[289,0.286],[290,0.2873],[291,0.2848],[294,0.2677],[295,0.2623],[296,0.264],[297,0.3035],[298,0.311],[299,0.3101],[300,0.3049],[301,0.2986],[302,0.2907],[304,0.2724],[305,0.2647],[306,0.2942],[307,0.2818],[308,0.2887],[309,0.2855],[310,0.2812],[313,0.2661],[314,0.2618],[315,0.2582],[316,0.2553],[317,0.2532],[318,0.2516],[319,0.2508],[320,0.2506]]
       }
       CTRL_R_eye.Y {
        type=MovieSceneScriptingFloatChannel
        default=0
        thinned_from=321
        keys [[0,0],[1,0.0007],[2,0.0025],[3,0.0046],[4,0.0072],[9,0.0226],[11,0.0283],[12,0.0307],[13,0.0328],[14,0.0345],[15,0.0357],[16,0.0364],[17,0.0366],[18,0.0176],[19,0.01],[20,0.0061],[21,0.0032],[23,-0.004],[24,-0.0079],[27,-0.0209],[28,-0.0249],[29,-0.0285],[30,-0.0319],[31,-0.0351],[33,-0.0407],[35,-0.0457],[37,-0.05],[38,-0.0646],[39,-0.07],[40,-0.0724],[42,-0.0754],[44,-0.0779],[46,-0.0797],[47,-0.0803],[48,-0.0539],[49,-0.0464],[50,-0.0446],[52,-0.0434],[54,-0.0417],[55,-0.0406],[57,-0.0377],[59,-0.034],[60,-0.032],[61,-0.0633],[62,-0.074],[63,-0.0769],[64,-0.0755],[66,-0.0699],[67,-0.0483],[68,-0.0236],[69,-0.0117],[70,-0.0065],[73,-0.0009],[75,0.0035],[76,0.0055],[77,0.0072],[78,0.0086],[79,0.0086],[80,0.0069],[81,0.004],[82,0.0003],[84,-0.0057],[86,-0.0112],[87,-0.0136],[88,-0.0157],[89,-0.0323],[90,-0.0375],[91,-0.041],[92,-0.043],[93,-0.0437],[95,-0.0438],[100,-0.0431],[102,-0.0423],[104,-0.041],[105,-0.0667],[106,-0.0743],[107,-0.0766],[108,-0.0772],[109,-0.077],[110,-0.0764],[111,-0.076],[112,-0.0764],[113,-0.0778],[114,-0.0976],[115,-0.1009],[116,-0.0995],[117,-0.0975],[118,-0.0959],[119,-0.0948],[120,-0.0942],[121,-0.094],[122,-0.0943],[123,-0.095],[124,-0.0961],[125,-0.0975],[126,-0.0993],[127,-0.1014],[128,-0.1039],[129,-0.107],[130,-0.1103],[132,-0.1177],[133,-0.1095],[134,-0.1053],[135,-0.1047],[136,-0.1129],[137,-0.1273],[138,-0.1414],[139,-0.1487],[140,-0.1501],[141,-0.1511],[142,-0.1516],[143,-0.1436],[144,-0.1169],[145,-0.0482],[146,-0.0199],[147,-0.0194],[148,-0.0199],[149,-0.0214],[150,-0.0234],[151,-0.0262],[153,-0.0327],[154,-0.0364],[155,-0.0411],[157,-0.0514],[158,-0.0564],[159,-0.0605],[160,-0.0638],[161,-0.0663],[162,-0.068],[163,-0.0693],[164,-0.0698],[165,-0.0698],[166,-0.0695],[167,-0.0686],[168,-0.0669],[169,-0.0648],[170,-0.0622],[171,-0.0588],[172,-0.0539],[173,-0.0471],[174,-0.0164],[176,0.0017],[177,0.0135],[178,0.0258],[179,0.0374],[180,0.0483],[181,0.0573],[182,-0.0798],[183,0.0676],[184,0.0638],[185,0.0732],[186,0.0487],[187,0.0578],[188,0.0674],[189,0.0773],[190,0.087],[191,0.0743],[192,0.0734],[193,0.0771],[194,0.0832],[195,0.0887],[196,0.0934],[197,0.0972],[198,0.1002],[199,0.1024],[200,0.1038],[201,0.1049],[202,0.1064],[203,0.1083],[204,0.1104],[205,0.1129],[206,0.1159],[207,0.1193],[208,0.097],[209,0.0936],[210,0.0962],[213,0.1083],[217,0.1233],[218,0.151],[219,0.1613],[220,0.1656],[221,0.1004],[222,0.1019],[223,0.102],[225,0.1016],[228,0.1004],[230,0.1006],[231,0.0995],[232,0.0953],[233,0.0883],[234,0.0787],[235,0.0668],[236,0.053],[237,0.0386],[238,0.0248],[239,0.0447],[240,-0.4343],[241,-0.2132],[242,-0.1293],[243,-0.1079],[244,-0.1159],[245,-0.1196],[246,-0.1143],[247,-0.1194],[249,-0.1318],[250,-0.1396],[251,-0.1418],[252,-0.1407],[254,-0.1375],[255,-0.1368],[261,-0.1356],[263,-0.1345],[264,-0.1335],[265,-0.1322],[266,-0.1305],[267,-0.1285],[268,-0.1262],[269,-0.095],[270,-0.0838],[271,-0.0779],[272,-0.073],[273,-0.0686],[276,-0.0561],[278,-0.0474],[282,-0.029],[283,-0.0248],[284,-0.0208],[285,-0.0171],[286,-0.0138],[287,-0.0111],[288,-0.0356],[289,-0.0411],[290,-0.0411],[293,-0.0372],[295,-0.0352],[296,-0.0337],[297,-0.0293],[300,-0.025],[301,-0.0238],[303,-0.0221],[305,-0.0207],[306,-0.0067],[307,-0.0023],[308,-0.0007],[310,0.0007],[311,0.0012],[313,0.0013],[318,0.0005],[320,-0.0001]]
       }
       CTRL_R_eye_blink {
        type=MovieSceneScriptingFloatChannel
        default=0
        thinned_from=272
        keys [[0,0.1532],[5,0.1525],[9,0.1506],[13,0.147],[17,0.142],[20,0.1372],[21,0.1354],[23,0.1309],[27,0.1211],[28,0.119],[29,0.1172],[30,0.1158],[32,0.1138],[34,0.1126],[36,0.112],[44,0.1128],[45,0.1138],[46,0.115],[47,0.1165],[48,0.1184],[49,0.1206],[50,0.1234],[51,0.1272],[52,0.1317],[53,0.1366],[54,0.1411],[55,0.1447],[56,0.1475],[57,0.1499],[58,0.1517],[59,0.1528],[60,0.1532],[63,0.1532],[64,0.1794],[65,0.2911],[66,0.564],[67,1],[68,1],[69,0.6138],[70,0.4114],[71,0.3121],[72,0.2528],[73,0.2043],[74,0.1735],[75,0.1583],[76,0.1532],[106,0.1532],[126,0.157],[128,0.1552],[131,0.1537],[139,0.1532],[140,0.1614],[141,0.2049],[142,0.3183],[143,0.5679],[144,1],[145,1],[146,0.6175],[147,0.3883],[148,0.2806],[149,0.2245],[150,0.1866],[151,0.167],[152,0.1558],[153,0.1534],[176,0.1532],[180,0.1514],[181,0.1764],[182,0.3109],[183,0.5553],[184,0.5704],[185,0.0803],[186,0.0458],[187,0.0493],[188,0.0628],[190,0.0997],[191,0.1128],[192,0.1156],[193,0.1118],[195,0.105],[196,0.1019],[197,0.099],[198,0.0964],[199,0.0942],[200,0.0924],[202,0.0897],[204,0.0877],[206,0.0864],[208,0.0856],[215,0.0854],[220,0.0862],[221,0.0868],[222,0.0881],[223,0.09],[224,0.0923],[226,0.0976],[227,0.1011],[230,0.1132],[231,0.1165],[232,0.1189],[233,0.1199],[234,0.1199],[235,0.1202],[236,0.1661],[237,0.3151],[238,0.5641],[239,1],[240,1],[241,0.6186],[242,0.386],[243,0.2668],[244,0.2022],[245,0.1646],[246,0.1444],[247,0.1354],[254,0.1358],[256,0.1365],[258,0.1439],[259,0.1473],[260,0.1498],[261,0.1515],[262,0.1525],[263,0.153],[320,0.1532]]
       }
       CTRL_C_eye.X {type=MovieSceneScriptingFloatChannel, default=0, const [0,321]}
       CTRL_C_eye.Y {type=MovieSceneScriptingFloatChannel, default=0, const [0,321]}
       CTRL_C_eyesConvergence {type=MovieSceneScriptingFloatChannel, default=0, const [0,321]}
       CTRL_L_eye_cheekRaise {type=MovieSceneScriptingFloatChannel, default=0, const [0.5722,321]}
       CTRL_L_eye_squintInner {
        type=MovieSceneScriptingFloatChannel
        default=0
        thinned_from=321
        keys [[0,0.1392],[5,0.1391],[10,0.1376],[12,0.1362],[16,0.132],[21,0.1259],[26,0.1183],[28,0.1159],[29,0.115],[32,0.1139],[44,0.1134],[49,0.1142],[51,0.1149],[53,0.1163],[60,0.1241],[62,0.1269],[66,0.1333],[68,0.1359],[69,0.1369],[72,0.1383],[75,0.139],[85,0.139],[102,0.1366],[105,0.1351],[109,0.1315],[117,0.1225],[126,0.1152],[132,0.1117],[145,0.1082],[160,0.1073],[180,0.1093],[189,0.1121],[193,0.1142],[198,0.1182],[204,0.1243],[211,0.1288],[218,0.1315],[238,0.1358],[248,0.1404],[269,0.1513],[288,0.1565],[297,0.1565],[300,0.156],[301,0.1555],[302,0.1545],[304,0.1514],[308,0.1442],[309,0.1428],[310,0.1418],[313,0.1401],[316,0.1393],[320,0.1392]]
       }
       CTRL_L_nose_wrinkleUpper {type=MovieSceneScriptingFloatChannel, default=0, const [1,321]}
       CTRL_R_nose_wrinkleUpper {type=MovieSceneScriptingFloatChannel, default=0, const [1,321]}
       CTRL_R_eye_squintInner {
        type=MovieSceneScriptingFloatChannel
        default=0
        thinned_from=321
        keys [[0,0.1392],[5,0.1391],[10,0.1376],[12,0.1362],[16,0.132],[21,0.1259],[26,0.1183],[28,0.1159],[29,0.115],[32,0.1139],[44,0.1134],[49,0.1142],[51,0.1149],[53,0.1163],[60,0.1241],[62,0.1269],[66,0.1333],[68,0.1359],[69,0.1369],[72,0.1383],[75,0.139],[85,0.139],[102,0.1366],[105,0.1351],[109,0.1315],[117,0.1225],[126,0.1152],[132,0.1117],[145,0.1082],[160,0.1073],[180,0.1093],[189,0.1121],[193,0.1142],[198,0.1182],[204,0.1243],[211,0.1288],[218,0.1315],[238,0.1358],[248,0.1404],[269,0.1513],[288,0.1565],[297,0.1565],[300,0.156],[301,0.1555],[302,0.1545],[304,0.1514],[308,0.1442],[309,0.1428],[310,0.1418],[313,0.1401],[316,0.1393],[320,0.1392]]
       }
       CTRL_R_eye_cheekRaise {type=MovieSceneScriptingFloatChannel, default=0, const [0.5722,321]}
       CTRL_C_eyelidsReduce {type=MovieSceneScriptingFloatChannel, default=0, const [0,321]}
       CTRL_R_brow_lateral {
        type=MovieSceneScriptingFloatChannel
        default=0
        thinned_from=321
        keys [[0,0],[52,0.0002],[140,0.0001],[141,0.0013],[142,0.0069],[143,0.0204],[144,0.0363],[145,0.0439],[146,0.0377],[147,0.0239],[148,0.0098],[149,0.0026],[150,0.0015],[151,0.0008],[152,0.0003],[153,0.0001],[192,0.0001],[221,0.0009],[226,0.0017],[231,0.0035],[232,0.0037],[235,0.0037],[236,0.0038],[237,0.0056],[238,0.0052],[239,0.0044],[240,0.004],[261,0.004],[285,0.0045],[288,0.0041],[296,0.0006],[299,0.0001],[320,0]]
       }
       CTRL_L_brow_lateral {
        type=MovieSceneScriptingFloatChannel
        default=0
        thinned_from=321
        keys [[0,0],[52,0.0002],[140,0.0001],[141,0.0013],[142,0.0069],[143,0.0204],[144,0.0363],[145,0.0439],[146,0.0377],[147,0.0239],[148,0.0098],[149,0.0026],[150,0.0015],[151,0.0008],[152,0.0003],[153,0.0001],[192,0.0001],[221,0.0009],[226,0.0017],[231,0.0035],[232,0.0037],[235,0.0037],[236,0.0038],[237,0.0056],[238,0.0052],[239,0.0044],[240,0.004],[261,0.004],[285,0.0045],[288,0.0041],[296,0.0006],[299,0.0001],[320,0]]
       }
       CTRL_R_brow_down {
        type=MovieSceneScriptingFloatChannel
        default=0
        thinned_from=321
        keys [[0,0],[21,0.0001],[46,0.0017],[53,0.0031],[58,0.0047],[59,0.0054],[61,0.0075],[62,0.0082],[63,0.0096],[64,0.0126],[65,0.0179],[66,0.054],[67,0.114],[68,0.1455],[69,0.1445],[70,0.1423],[71,0.1405],[73,0.1383],[74,0.1375],[75,0.1372],[76,0.1381],[77,0.1402],[79,0.1454],[80,0.1469],[82,0.1482],[95,0.1506],[135,0.1505],[136,0.1496],[137,0.1475],[140,0.1391],[141,0.1379],[142,0.1339],[143,0.1197],[144,0.0985],[145,0.0788],[147,0.045],[148,0.0305],[149,0.0202],[150,0.0142],[151,0.0105],[152,0.0084],[153,0.0072],[154,0.0065],[155,0.0056],[157,0.0036],[159,0.0022],[162,0.0011],[167,0.0003],[182,0],[186,0.0003],[211,0.0048],[217,0.0051],[220,0.0057],[221,0.0062],[222,0.0069],[224,0.0088],[226,0.0112],[227,0.0132],[230,0.0203],[231,0.0222],[232,0.0235],[233,0.024],[235,0.0241],[236,0.0245],[237,0.0358],[238,0.0632],[239,0.0926],[240,0.1063],[241,0.0948],[242,0.0708],[243,0.0503],[244,0.0369],[245,0.0248],[246,0.0161],[247,0.0127],[259,0.0131],[263,0.0141],[270,0.0174],[274,0.0199],[277,0.0205],[284,0.0207],[286,0.0201],[287,0.0195],[288,0.0186],[289,0.0173],[290,0.0155],[294,0.0065],[295,0.0045],[296,0.003],[297,0.0019],[298,0.0011],[299,0.0006],[301,0.0002],[320,0]]
       }
       CTRL_L_brow_down {
        type=MovieSceneScriptingFloatChannel
        default=0
        thinned_from=321
        keys [[0,0],[21,0.0001],[46,0.0017],[53,0.0031],[58,0.0047],[59,0.0054],[61,0.0075],[62,0.0082],[63,0.0096],[64,0.0126],[65,0.0179],[66,0.054],[67,0.1138],[68,0.1453],[69,0.1443],[70,0.1422],[71,0.1404],[73,0.1382],[74,0.1375],[75,0.1372],[76,0.1381],[77,0.1402],[79,0.1454],[80,0.1469],[82,0.1482],[95,0.1506],[135,0.1505],[136,0.1496],[137,0.1475],[140,0.1391],[141,0.1379],[142,0.1339],[143,0.1197],[144,0.0985],[145,0.0787],[147,0.045],[148,0.0305],[149,0.0202],[150,0.0142],[151,0.0105],[152,0.0084],[153,0.0072],[154,0.0065],[155,0.0056],[157,0.0036],[159,0.0022],[162,0.0011],[167,0.0003],[182,0],[186,0.0003],[211,0.0048],[217,0.0051],[220,0.0057],[221,0.0062],[222,0.0069],[224,0.0088],[226,0.0112],[227,0.0132],[230,0.0203],[231,0.0222],[232,0.0235],[233,0.024],[235,0.0241],[236,0.0245],[237,0.0358],[238,0.0632],[239,0.0926],[240,0.1063],[241,0.0948],[242,0.0708],[243,0.0503],[244,0.0369],[245,0.0248],[246,0.0161],[247,0.0127],[259,0.0131],[263,0.0141],[270,0.0174],[274,0.0199],[277,0.0205],[284,0.0207],[286,0.0201],[287,0.0195],[288,0.0186],[289,0.0173],[290,0.0155],[294,0.0065],[295,0.0045],[296,0.003],[297,0.0019],[298,0.0011],[299,0.0006],[301,0.0002],[320,0]]
       }
       CTRL_L_brow_raiseOut {
        type=MovieSceneScriptingFloatChannel
        default=0
        thinned_from=303
        keys [[0,0.3519],[9,0.3547],[17,0.3636],[21,0.3705],[28,0.3875],[31,0.392],[37,0.3949],[44,0.3943],[47,0.3915],[49,0.3877],[50,0.3849],[51,0.3811],[54,0.3661],[56,0.3579],[60,0.3451],[61,0.3431],[63,0.3411],[64,0.3384],[65,0.333],[66,0.2801],[67,0.1894],[68,0.1413],[69,0.1495],[70,0.1665],[71,0.1809],[72,0.1901],[73,0.1983],[74,0.2043],[75,0.2065],[76,0.2056],[77,0.2033],[79,0.1979],[80,0.1962],[83,0.1944],[135,0.1925],[136,0.1938],[137,0.1968],[138,0.2003],[139,0.2033],[140,0.2045],[141,0.204],[142,0.2008],[143,0.1842],[144,0.1592],[145,0.1463],[146,0.1725],[147,0.2312],[148,0.2923],[149,0.326],[150,0.3342],[151,0.3394],[152,0.3423],[158,0.3494],[179,0.3533],[182,0.3562],[185,0.3617],[186,0.3646],[188,0.3726],[192,0.3912],[196,0.4051],[199,0.4128],[204,0.4194],[220,0.4207],[221,0.4198],[222,0.4179],[224,0.4116],[226,0.4035],[227,0.3971],[230,0.375],[231,0.3689],[232,0.3647],[233,0.3631],[235,0.3628],[236,0.3617],[237,0.344],[238,0.2883],[239,0.2275],[240,0.1989],[241,0.216],[242,0.2516],[243,0.2819],[244,0.3016],[245,0.3193],[246,0.3322],[247,0.3373],[286,0.3419],[297,0.351],[320,0.3519]]
       }
       CTRL_L_brow_raiseIn {
        type=MovieSceneScriptingFloatChannel
        default=0
        thinned_from=303
        keys [[0,0.1471],[6,0.1479],[10,0.1497],[15,0.1537],[20,0.1594],[22,0.1625],[27,0.1718],[29,0.1748],[31,0.1767],[34,0.1783],[42,0.1788],[44,0.1782],[47,0.1758],[49,0.1731],[50,0.1713],[51,0.1689],[54,0.1598],[55,0.1574],[57,0.1541],[58,0.1529],[59,0.1521],[60,0.1519],[61,0.1528],[62,0.1548],[63,0.1573],[64,0.1594],[65,0.1603],[66,0.1588],[67,0.156],[68,0.1545],[69,0.1651],[70,0.1875],[71,0.2071],[72,0.2202],[73,0.2321],[74,0.2418],[75,0.2481],[76,0.2515],[77,0.2534],[84,0.2568],[135,0.2579],[136,0.2573],[137,0.2559],[140,0.2495],[141,0.2482],[142,0.243],[143,0.2254],[144,0.2008],[145,0.182],[146,0.1728],[147,0.1673],[148,0.1641],[150,0.1585],[151,0.1563],[152,0.1546],[154,0.1522],[157,0.1498],[161,0.1481],[176,0.1471],[179,0.1478],[181,0.149],[183,0.151],[185,0.1539],[186,0.1561],[187,0.1588],[189,0.1655],[191,0.1727],[192,0.176],[195,0.1841],[197,0.1887],[199,0.1924],[202,0.1958],[206,0.1984],[216,0.1991],[220,0.1986],[221,0.1981],[222,0.1971],[224,0.1939],[226,0.1897],[227,0.1866],[230,0.1758],[231,0.1729],[232,0.1709],[233,0.1701],[235,0.1699],[236,0.1694],[237,0.1644],[238,0.1453],[239,0.1243],[240,0.1143],[241,0.1181],[242,0.1259],[243,0.1325],[244,0.1366],[245,0.1402],[246,0.143],[247,0.1445],[249,0.1457],[252,0.1464],[262,0.1482],[270,0.1523],[274,0.1548],[278,0.1556],[286,0.1555],[288,0.1548],[291,0.1526],[295,0.1489],[297,0.1479],[320,0.147]]
       }
       CTRL_R_brow_raiseIn {
        type=MovieSceneScriptingFloatChannel
        default=0
        thinned_from=303
        keys [[0,0.1612],[6,0.162],[10,0.1638],[15,0.1679],[20,0.1737],[22,0.1769],[27,0.1864],[29,0.1894],[31,0.1914],[35,0.1933],[42,0.1936],[44,0.193],[47,0.1904],[49,0.1877],[50,0.1858],[51,0.1833],[54,0.1738],[55,0.1714],[57,0.1679],[58,0.1667],[59,0.1659],[60,0.1656],[61,0.1665],[62,0.1684],[63,0.1708],[64,0.1727],[65,0.1735],[66,0.1717],[67,0.1681],[68,0.1663],[69,0.1765],[70,0.198],[71,0.2167],[72,0.2293],[73,0.2408],[74,0.25],[75,0.256],[76,0.2592],[77,0.261],[85,0.2643],[135,0.2652],[136,0.2646],[138,0.2615],[140,0.2573],[141,0.256],[142,0.2511],[143,0.2341],[144,0.2101],[145,0.1919],[146,0.1835],[147,0.179],[150,0.172],[151,0.17],[153,0.167],[156,0.1643],[160,0.1624],[177,0.1612],[180,0.1624],[182,0.164],[184,0.1665],[185,0.1681],[186,0.1703],[187,0.1731],[192,0.1907],[195,0.1989],[197,0.2037],[199,0.2074],[202,0.211],[206,0.2136],[218,0.2142],[221,0.2134],[223,0.211],[226,0.2053],[227,0.2025],[230,0.1924],[231,0.1897],[232,0.1878],[233,0.1871],[235,0.187],[236,0.1865],[237,0.1828],[238,0.1674],[239,0.1503],[240,0.1423],[241,0.1445],[242,0.149],[243,0.1528],[244,0.1552],[245,0.1572],[246,0.1588],[247,0.1598],[249,0.1608],[262,0.1633],[270,0.1678],[274,0.1706],[278,0.1715],[286,0.1713],[288,0.1705],[290,0.1689],[295,0.1634],[297,0.1621],[320,0.1611]]
       }
       CTRL_R_brow_raiseOut {
        type=MovieSceneScriptingFloatChannel
        default=0
        thinned_from=303
        keys [[0,0],[3,0.0003],[5,0.0009],[7,0.0019],[9,0.0035],[11,0.0056],[13,0.0082],[15,0.0113],[17,0.0149],[19,0.019],[20,0.0212],[21,0.0236],[22,0.0265],[26,0.0395],[27,0.0426],[28,0.0453],[29,0.0477],[30,0.0496],[31,0.051],[32,0.0522],[33,0.0531],[35,0.0542],[37,0.0547],[42,0.0548],[44,0.0539],[45,0.0529],[46,0.0517],[47,0.0503],[48,0.0485],[49,0.0462],[50,0.0435],[51,0.0399],[52,0.0354],[53,0.0307],[54,0.0263],[55,0.0228],[56,0.0201],[57,0.0178],[58,0.016],[59,0.0149],[60,0.0145],[61,0.016],[62,0.0201],[63,0.026],[64,0.0329],[65,0.0401],[66,0.0447],[67,0.0515],[68,0.0714],[69,0.1162],[70,0.1741],[71,0.2212],[72,0.2513],[73,0.2749],[74,0.2932],[75,0.307],[76,0.3166],[77,0.3224],[78,0.3255],[82,0.3317],[98,0.3376],[135,0.3366],[136,0.3348],[137,0.3305],[138,0.3245],[140,0.3113],[141,0.3077],[142,0.2943],[143,0.2592],[144,0.2142],[145,0.1725],[146,0.1356],[147,0.0991],[148,0.0675],[149,0.0453],[150,0.0321],[151,0.0238],[152,0.019],[153,0.0164],[154,0.0147],[155,0.0126],[156,0.01],[157,0.008],[158,0.0063],[159,0.0049],[160,0.0039],[162,0.0024],[166,0.0009],[170,0.0001],[176,0],[178,0.0006],[179,0.0012],[180,0.0021],[181,0.0032],[182,0.0048],[183,0.0066],[184,0.009],[185,0.0118],[186,0.0155],[187,0.0202],[188,0.0258],[191,0.0441],[192,0.0499],[193,0.0549],[195,0.0638],[196,0.0679],[197,0.0718],[198,0.0752],[199,0.0781],[200,0.0805],[201,0.0824],[202,0.0841],[203,0.0855],[204,0.0868],[205,0.0877],[207,0.0891],[209,0.0898],[217,0.0898],[220,0.0895],[221,0.0892],[222,0.0886],[224,0.0866],[226,0.0842],[230,0.0771],[231,0.0755],[232,0.0743],[233,0.0735],[236,0.0733],[237,0.0793],[238,0.092],[239,0.1056],[240,0.1119],[241,0.0989],[242,0.0721],[243,0.0493],[244,0.0345],[245,0.0212],[246,0.0116],[247,0.0079],[251,0.008],[259,0.0096],[261,0.0105],[262,0.0112],[263,0.0122],[265,0.0147],[269,0.0207],[270,0.0225],[273,0.0285],[274,0.03],[275,0.0311],[277,0.0322],[279,0.0326],[283,0.0328],[285,0.0323],[286,0.0318],[287,0.0308],[288,0.0294],[289,0.0273],[290,0.0244],[291,0.021],[293,0.0137],[294,0.0102],[295,0.0071],[296,0.0046],[297,0.0028],[298,0.0016],[320,0.0001]]
       }
       CTRL_lookAtSwitch {type=MovieSceneScriptingFloatChannel, default=0, const [0,321]}
       CTRL_C_teethD.X {type=MovieSceneScriptingFloatChannel, default=0, const [0,321]}
       CTRL_C_teethD.Y {type=MovieSceneScriptingFloatChannel, default=0, const [0,321]}
       CTRL_C_teethU.X {type=MovieSceneScriptingFloatChannel, default=0, const [0,321]}
       CTRL_C_teethU.Y {type=MovieSceneScriptingFloatChannel, default=0, const [0,321]}
       CTRL_C_teeth_fwdBackU {type=MovieSceneScriptingFloatChannel, default=0, const [0,321]}
       CTRL_L_mouth_corner.X {type=MovieSceneScriptingFloatChannel, default=0, const [0,321]}
       CTRL_L_mouth_corner.Y {
        type=MovieSceneScriptingFloatChannel
        default=0
        thinned_from=321
        keys [[0,0.4891],[7,0.4871],[10,0.4835],[12,0.4788],[16,0.464],[21,0.4426],[26,0.4158],[28,0.4072],[29,0.4042],[32,0.4004],[45,0.3989],[50,0.4023],[52,0.4061],[54,0.4121],[61,0.4407],[67,0.4733],[69,0.4811],[72,0.4862],[100,0.4815],[103,0.4786],[105,0.4747],[109,0.462],[117,0.4305],[126,0.405],[132,0.3926],[145,0.3801],[157,0.3771],[179,0.3834],[188,0.3924],[192,0.3992],[197,0.4122],[204,0.4369],[211,0.4528],[218,0.4622],[238,0.4773],[248,0.4934],[269,0.5317],[288,0.5502],[300,0.5482],[301,0.5464],[302,0.5429],[304,0.5322],[308,0.5068],[309,0.5019],[310,0.4983],[313,0.4922],[318,0.4891],[320,0.4891]]
       }
       CTRL_R_mouth_corner.X {type=MovieSceneScriptingFloatChannel, default=0, const [0,321]}
       CTRL_R_mouth_corner.Y {
        type=MovieSceneScriptingFloatChannel
        default=0
        thinned_from=321
        keys [[0,0.2715],[7,0.2705],[10,0.2684],[12,0.2658],[16,0.2576],[21,0.2457],[26,0.2309],[28,0.2261],[29,0.2244],[32,0.2223],[45,0.2215],[50,0.2234],[52,0.2255],[54,0.2288],[61,0.2447],[67,0.2628],[69,0.2671],[72,0.2699],[83,0.2715],[102,0.2665],[105,0.2636],[109,0.2565],[117,0.239],[126,0.2249],[132,0.218],[145,0.2111],[158,0.2094],[179,0.2129],[188,0.2178],[192,0.2216],[197,0.2288],[204,0.2426],[211,0.2514],[218,0.2566],[238,0.265],[248,0.274],[269,0.2952],[288,0.3055],[300,0.3043],[301,0.3034],[302,0.3014],[304,0.2955],[308,0.2814],[309,0.2787],[310,0.2767],[313,0.2733],[317,0.2717],[320,0.2715]]
       }
       CTRL_C_teeth_fwdBackD {type=MovieSceneScriptingFloatChannel, default=0, const [0,321]}
       CTRL_R_mouth_cornerSharpnessD {
        type=MovieSceneScriptingFloatChannel
        default=0
        thinned_from=321
        keys [[0,0.3137],[8,0.3123],[11,0.31],[14,0.3055],[20,0.2923],[23,0.2836],[27,0.2704],[29,0.2649],[31,0.2612],[33,0.2589],[49,0.2515],[55,0.2447],[62,0.2339],[68,0.2231],[70,0.2204],[73,0.2179],[102,0.1999],[104,0.1978],[106,0.1948],[110,0.1866],[116,0.1728],[121,0.1637],[127,0.1548],[131,0.1501],[135,0.1469],[146,0.1423],[152,0.1411],[157,0.141],[175,0.1458],[179,0.1481],[184,0.1529],[187,0.1568],[190,0.1618],[192,0.1661],[194,0.1715],[197,0.1809],[198,0.1845],[203,0.2054],[204,0.2092],[206,0.2153],[209,0.2229],[213,0.231],[216,0.2357],[234,0.25],[242,0.2615],[247,0.2712],[256,0.2941],[270,0.3186],[295,0.3421],[300,0.3438],[301,0.3433],[302,0.3418],[304,0.3369],[308,0.3243],[310,0.3199],[314,0.3156],[320,0.3137]]
       }
       CTRL_R_mouth_cornerSharpnessU {
        type=MovieSceneScriptingFloatChannel
        default=0
        thinned_from=321
        keys [[0,0.0651],[6,0.0651],[13,0.0641],[21,0.0614],[31,0.0563],[39,0.0535],[46,0.0503],[48,0.0488],[49,0.0478],[50,0.0465],[51,0.0449],[52,0.0431],[54,0.0389],[61,0.0222],[65,0.0114],[66,0.0088],[67,0.0065],[68,0.0045],[69,0.0029],[70,0.0019],[72,0.0006],[75,-0.0004],[158,-0.0007],[164,-0.0003],[178,0.0015],[186,0.0042],[190,0.0064],[193,0.0087],[197,0.0128],[203,0.0211],[205,0.0234],[209,0.0269],[214,0.0302],[218,0.032],[234,0.0361],[242,0.0399],[247,0.0432],[256,0.0509],[269,0.0587],[297,0.0685],[299,0.0689],[302,0.0687],[316,0.0654],[320,0.0651]]
       }
       CTRL_L_eye_eyelidU.X {type=MovieSceneScriptingFloatChannel, default=0, const [0,321]}
       CTRL_L_eye_eyelidU.Y {
        type=MovieSceneScriptingFloatChannel
        default=0
        thinned_from=321
        keys [[0,0],[67,0],[68,-0.1354],[69,-0.0593],[70,-0.0212],[71,-0.0056],[72,0],[144,0],[145,-0.1126],[146,-0.0453],[147,-0.0229],[148,-0.0117],[149,-0.003],[150,0],[180,0],[181,-0.1246],[182,-0.5652],[183,-1],[184,-1],[185,-0.86],[186,-0.5404],[187,-0.3484],[188,-0.203],[189,-0.0833],[190,-0.013],[191,-0.0039],[192,-0.0005],[193,0],[238,0],[239,-0.1493],[240,-0.2687],[241,0],[320,0]]
       }
       CTRL_L_mouth_cornerSharpnessD {
        type=MovieSceneScriptingFloatChannel
        default=0
        thinned_from=321
        keys [[0,0.3137],[8,0.3123],[11,0.31],[14,0.3055],[20,0.2923],[23,0.2836],[27,0.2704],[29,0.2649],[31,0.2612],[33,0.2589],[49,0.2515],[55,0.2447],[62,0.2339],[68,0.2231],[70,0.2204],[73,0.2179],[102,0.1999],[104,0.1978],[106,0.1948],[110,0.1866],[116,0.1728],[121,0.1637],[127,0.1548],[131,0.1501],[135,0.1469],[146,0.1423],[152,0.1411],[157,0.141],[175,0.1458],[179,0.1481],[184,0.1529],[187,0.1568],[190,0.1618],[192,0.1661],[194,0.1715],[197,0.1809],[198,0.1845],[203,0.2054],[204,0.2092],[206,0.2153],[209,0.2229],[213,0.231],[216,0.2357],[234,0.25],[242,0.2615],[247,0.2712],[256,0.2941],[270,0.3186],[295,0.3421],[300,0.3438],[301,0.3433],[302,0.3418],[304,0.3369],[308,0.3243],[310,0.3199],[314,0.3156],[320,0.3137]]
       }
       CTRL_L_eye_eyelidD.X {type=MovieSceneScriptingFloatChannel, default=0, const [0,321]}
       CTRL_L_eye_eyelidD.Y {
        type=MovieSceneScriptingFloatChannel
        default=0
        thinned_from=321
        keys [[0,0],[64,0],[65,0.0393],[66,0.1048],[67,0.2358],[68,0.1003],[69,0.038],[70,0.0068],[71,0.0009],[72,0],[140,0],[141,0.0098],[142,0.0368],[143,0.1032],[144,0.2358],[145,0.1232],[146,0.0454],[147,0.0187],[148,0.0065],[149,0.0008],[150,0],[182,0],[183,0.0526],[184,0.0701],[185,0.0214],[186,0.0094],[187,0.0045],[188,0.0019],[189,0.0002],[190,0],[320,0]]
       }
       CTRL_L_mouth_cornerSharpnessU {
        type=MovieSceneScriptingFloatChannel
        default=0
        thinned_from=321
        keys [[0,0.0651],[5,0.0651],[11,0.064],[21,0.0589],[27,0.0547],[29,0.0538],[32,0.0533],[44,0.053],[49,0.0539],[51,0.0547],[53,0.0563],[55,0.0584],[61,0.0663],[66,0.075],[67,0.0766],[68,0.078],[69,0.0791],[70,0.0798],[73,0.081],[76,0.0815],[84,0.0816],[94,0.0809],[100,0.0804],[104,0.0796],[108,0.0777],[118,0.0713],[130,0.0661],[144,0.0635],[166,0.0629],[185,0.0636],[214,0.0685],[237,0.0698],[265,0.0734],[295,0.075],[300,0.0739],[302,0.0728],[308,0.0674],[310,0.0663],[314,0.0653],[320,0.0651]]
       }
       CTRL_R_mouth_lipsTowardsTeethD {type=MovieSceneScriptingFloatChannel, default=0, const [0,321]}
       CTRL_L_mouth_lipsTowardsTeethD {type=MovieSceneScriptingFloatChannel, default=0, const [0,321]}
       CTRL_R_eye_eyelidU.X {type=MovieSceneScriptingFloatChannel, default=0, const [0,321]}
       CTRL_R_eye_eyelidU.Y {
        type=MovieSceneScriptingFloatChannel
        default=0
        thinned_from=321
        keys [[0,0],[67,0],[68,-0.1354],[69,-0.0593],[70,-0.0212],[71,-0.0056],[72,0],[144,0],[145,-0.1126],[146,-0.0453],[147,-0.0229],[148,-0.0117],[149,-0.003],[150,0],[180,0],[181,-0.1246],[182,-0.5652],[183,-1],[184,-1],[185,-0.86],[186,-0.5404],[187,-0.3484],[188,-0.203],[189,-0.1122],[190,-0.0507],[191,-0.0195],[192,-0.0041],[193,0],[238,0],[239,-0.1493],[240,-0.2687],[241,0],[320,0]]
       }
       CTRL_L_mouth_lipsTowardsTeethU {type=MovieSceneScriptingFloatChannel, default=0, const [0,321]}
       CTRL_R_eye_eyelidD.X {type=MovieSceneScriptingFloatChannel, default=0, const [0,321]}
       CTRL_R_eye_eyelidD.Y {
        type=MovieSceneScriptingFloatChannel
        default=0
        thinned_from=321
        keys [[0,0],[64,0],[65,0.0393],[66,0.1048],[67,0.2358],[68,0.1003],[69,0.038],[70,0.0068],[71,0.0009],[72,0],[140,0],[141,0.0098],[142,0.0368],[143,0.1032],[144,0.2358],[145,0.1232],[146,0.0454],[147,0.0187],[148,0.0065],[149,0.0008],[150,0],[182,0],[183,0.0526],[184,0.0701],[185,0.0214],[186,0.0094],[187,0.0045],[188,0.0019],[189,0.0002],[190,0],[320,0]]
       }
       CTRL_R_mouth_lipsTowardsTeethU {type=MovieSceneScriptingFloatChannel, default=0, const [0,321]}
       CTRL_faceGUIfollowHead {type=MovieSceneScriptingFloatChannel, default=1, const [1,321]}
       CTRL_eyesAimFollowHead {type=MovieSceneScriptingFloatChannel, default=0, const [0,321]}
       CTRL_C_nose.X {type=MovieSceneScriptingFloatChannel, default=0, const [0,321]}
       CTRL_C_nose.Y {type=MovieSceneScriptingFloatChannel, default=0, const [0,321]}
       CTRL_C_teeth_rotateU {type=MovieSceneScriptingFloatChannel, default=0, const [0,321]}
       CTRL_C_teeth_rotateD {type=MovieSceneScriptingFloatChannel, default=0, const [0,321]}
       CTRL_L_brow_inner {
        type=MovieSceneScriptingFloatChannel
        default=0
        thinned_from=302
        keys [[0,0],[176,0],[177,0.0003],[178,0.001],[179,0.0024],[180,0.0046],[181,0.009],[182,0.016],[183,0.0247],[184,0.0206],[185,0.0048],[186,-0.0002],[187,0.0127],[188,0.0321],[189,0.0525],[190,0.0678],[191,0.0764],[192,0.0816],[193,0.0851],[194,0.0879],[195,0.0903],[196,0.0921],[197,0.0932],[198,0.0938],[199,0.094],[224,0.094],[228,0.0932],[231,0.0918],[233,0.0904],[235,0.0884],[236,0.0862],[237,0.0822],[238,0.0647],[239,0.0351],[240,0.0132],[241,0.0028],[242,-0.0058],[243,-0.0131],[244,-0.0183],[245,-0.0209],[246,-0.0222],[247,-0.0239],[248,-0.0261],[249,-0.028],[250,-0.0297],[251,-0.0309],[252,-0.0317],[253,-0.0319],[254,-0.0315],[256,-0.0292],[260,-0.0255],[261,-0.0242],[262,-0.0226],[263,-0.0206],[264,-0.0183],[266,-0.0132],[269,-0.0047],[270,-0.0016],[272,0.0055],[273,0.0088],[274,0.0116],[275,0.0134],[276,0.0145],[277,0.0153],[278,0.0158],[280,0.0163],[284,0.0163],[286,0.0159],[287,0.0154],[288,0.0147],[289,0.0136],[290,0.0122],[294,0.005],[295,0.0035],[296,0.0022],[297,0.0015],[298,0.0012],[320,-0.0001]]
       }
       CTRL_R_brow_inner {
        type=MovieSceneScriptingFloatChannel
        default=0
        thinned_from=302
        keys [[0,0],[176,0],[177,0.0003],[178,0.001],[179,0.0024],[180,0.0046],[181,0.009],[182,0.016],[183,0.0247],[184,0.0206],[185,0.0048],[186,-0.0002],[187,0.0127],[188,0.0321],[189,0.0525],[190,0.0678],[191,0.0764],[192,0.0816],[193,0.0851],[194,0.0879],[195,0.0903],[196,0.0921],[197,0.0932],[198,0.0938],[199,0.094],[224,0.094],[228,0.0932],[231,0.0918],[233,0.0904],[235,0.0884],[236,0.0862],[237,0.0822],[238,0.0647],[239,0.0351],[240,0.0132],[241,0.0028],[242,-0.0058],[243,-0.0131],[244,-0.0183],[245,-0.0209],[246,-0.0222],[247,-0.0239],[248,-0.0261],[249,-0.028],[250,-0.0297],[251,-0.0309],[252,-0.0317],[253,-0.0319],[254,-0.0315],[256,-0.0292],[260,-0.0255],[261,-0.0242],[262,-0.0226],[263,-0.0206],[264,-0.0183],[266,-0.0132],[269,-0.0047],[270,-0.0016],[272,0.0055],[273,0.0088],[274,0.0116],[275,0.0134],[276,0.0145],[277,0.0153],[278,0.0158],[280,0.0163],[284,0.0163],[286,0.0159],[287,0.0154],[288,0.0147],[289,0.0136],[290,0.0122],[294,0.005],[295,0.0035],[296,0.0022],[297,0.0015],[298,0.0012],[320,-0.0001]]
       }
       CTRL_L_brow_outer {
        type=MovieSceneScriptingFloatChannel
        default=0
        thinned_from=302
        keys [[0,0],[175,0],[176,0.0001],[177,0.0011],[178,0.0038],[179,0.0091],[180,0.0177],[181,0.0344],[182,0.061],[183,0.0943],[184,0.1165],[185,0.1263],[186,0.145],[187,0.1807],[188,0.2221],[189,0.2616],[190,0.2882],[191,0.3025],[193,0.3252],[194,0.3359],[195,0.3451],[196,0.3521],[197,0.3563],[198,0.3585],[224,0.359],[227,0.3572],[230,0.3531],[232,0.3485],[234,0.342],[235,0.338],[236,0.3297],[237,0.3143],[238,0.2879],[239,0.2475],[240,0.1961],[241,0.1268],[242,0.0496],[243,-0.0074],[244,-0.0405],[245,-0.0646],[246,-0.0831],[247,-0.0997],[248,-0.1144],[249,-0.1267],[250,-0.1364],[251,-0.1437],[252,-0.1485],[253,-0.151],[254,-0.1511],[255,-0.1489],[256,-0.1458],[259,-0.1382],[260,-0.1352],[261,-0.1316],[262,-0.1269],[263,-0.1213],[264,-0.1149],[265,-0.1079],[266,-0.1003],[267,-0.0924],[269,-0.0759],[270,-0.0667],[271,-0.0561],[272,-0.0452],[273,-0.0347],[274,-0.0256],[275,-0.0188],[276,-0.0143],[277,-0.0113],[278,-0.0094],[279,-0.0083],[281,-0.007],[283,-0.0051],[284,-0.0046],[288,-0.0036],[295,-0.0007],[298,-0.0001],[320,-0.0001]]
       }
       CTRL_R_brow_outer {
        type=MovieSceneScriptingFloatChannel
        default=0
        thinned_from=302
        keys [[0,0],[176,0],[177,0.0006],[178,0.0021],[179,0.0049],[180,0.0095],[181,0.0185],[182,0.0329],[183,0.0508],[184,0.0561],[185,0.0491],[186,0.0525],[187,0.0744],[188,0.103],[189,0.1311],[190,0.1501],[191,0.1603],[192,0.1684],[193,0.1751],[194,0.1808],[195,0.1858],[196,0.1896],[197,0.1919],[198,0.193],[199,0.1935],[224,0.1933],[227,0.1924],[230,0.1901],[232,0.1876],[234,0.1842],[235,0.182],[236,0.1775],[237,0.1692],[238,0.1501],[239,0.1189],[240,0.0853],[241,0.0489],[242,0.0116],[243,-0.0134],[244,-0.027],[245,-0.0378],[246,-0.0469],[247,-0.0557],[248,-0.0637],[249,-0.0703],[250,-0.0756],[251,-0.0795],[252,-0.0821],[253,-0.0835],[254,-0.0835],[255,-0.0824],[256,-0.0807],[259,-0.0766],[260,-0.075],[261,-0.073],[262,-0.0706],[263,-0.0676],[264,-0.0642],[265,-0.0606],[266,-0.0566],[268,-0.0481],[269,-0.0436],[270,-0.0386],[271,-0.0329],[272,-0.0269],[273,-0.0211],[274,-0.0161],[275,-0.0124],[276,-0.0098],[277,-0.0081],[278,-0.0069],[279,-0.0062],[283,-0.004],[296,-0.0004],[299,-0.0001],[320,-0.0001]]
       }
       CTRL_C_mouth.X {type=MovieSceneScriptingFloatChannel, default=0, const [0.0923,321]}
       CTRL_C_mouth.Y {
        type=MovieSceneScriptingFloatChannel
        default=0
        thinned_from=321
        keys [[0,0.1096],[4,0.1096],[9,0.1087],[11,0.1079],[14,0.1058],[21,0.0992],[26,0.0932],[28,0.0912],[29,0.0906],[33,0.0896],[45,0.0893],[50,0.0902],[52,0.091],[55,0.0931],[61,0.0988],[67,0.1061],[69,0.1078],[72,0.1089],[75,0.1094],[84,0.1096],[102,0.1075],[105,0.1064],[109,0.1035],[117,0.0965],[127,0.0902],[134,0.0873],[148,0.0848],[161,0.0846],[176,0.0855],[188,0.0879],[193,0.09],[198,0.0931],[204,0.0979],[211,0.1015],[218,0.1036],[238,0.1069],[248,0.1106],[269,0.1191],[288,0.1233],[297,0.1233],[300,0.1228],[301,0.1224],[302,0.1216],[304,0.1193],[308,0.1136],[309,0.1125],[310,0.1117],[313,0.1103],[316,0.1097],[320,0.1096]]
       }
       CTRL_L_mouth_upperLipRaise {
        type=MovieSceneScriptingFloatChannel
        default=0
        thinned_from=321
        keys [[0,0.2355],[7,0.2346],[10,0.2328],[12,0.2305],[16,0.2234],[21,0.2131],[26,0.2002],[28,0.1961],[29,0.1946],[32,0.1928],[45,0.1921],[50,0.1937],[52,0.1955],[54,0.1984],[61,0.2122],[67,0.2279],[69,0.2317],[72,0.2341],[83,0.2354],[102,0.2311],[105,0.2286],[109,0.2225],[117,0.2073],[126,0.195],[132,0.189],[145,0.183],[158,0.1816],[179,0.1846],[188,0.1889],[192,0.1922],[197,0.1985],[204,0.2104],[211,0.218],[218,0.2225],[238,0.2298],[248,0.2376],[269,0.256],[288,0.2649],[300,0.2639],[301,0.2631],[302,0.2614],[304,0.2563],[308,0.244],[309,0.2417],[310,0.24],[313,0.237],[317,0.2356],[320,0.2355]]
       }
       CTRL_L_mouth_cornerPull {
        type=MovieSceneScriptingFloatChannel
        default=0
        thinned_from=321
        keys [[0,0.0721],[6,0.072],[11,0.071],[21,0.0653],[27,0.0607],[29,0.0596],[32,0.0591],[43,0.0588],[46,0.0594],[48,0.0602],[50,0.0615],[51,0.0625],[52,0.0641],[53,0.066],[54,0.0682],[55,0.0708],[60,0.0851],[61,0.0882],[62,0.0917],[66,0.1073],[67,0.1108],[68,0.1138],[69,0.1162],[70,0.1178],[72,0.1197],[74,0.1209],[77,0.1216],[84,0.1217],[100,0.1204],[104,0.1194],[107,0.1177],[118,0.1087],[131,0.1013],[150,0.0956],[158,0.0943],[183,0.0928],[213,0.0881],[293,0.0857],[297,0.0848],[300,0.0833],[301,0.0825],[303,0.0805],[307,0.0757],[308,0.0747],[309,0.0739],[311,0.0731],[315,0.0723],[320,0.0722]]
       }
       CTRL_L_mouth_dimple {
        type=MovieSceneScriptingFloatChannel
        default=0
        thinned_from=321
        keys [[0,0],[43,0],[47,0.0006],[50,0.0017],[51,0.0023],[52,0.0033],[53,0.0045],[55,0.0074],[60,0.0163],[61,0.0183],[63,0.0229],[66,0.0302],[67,0.0323],[68,0.0342],[69,0.0357],[70,0.0367],[72,0.0379],[75,0.0388],[92,0.0391],[104,0.0388],[133,0.0343],[138,0.0339],[186,0.0269],[195,0.0239],[205,0.0186],[216,0.0153],[222,0.0145],[223,0.0145],[227,0.014],[232,0.0136],[281,0.0044],[288,0.0039],[308,0.0001],[320,0]]
       }
       CTRL_L_mouth_cornerDepress {type=MovieSceneScriptingFloatChannel, default=0, const [0,321]}
       CTRL_L_mouth_stretch {type=MovieSceneScriptingFloatChannel, default=0, const [0,321]}
       CTRL_L_mouth_lowerLipDepress {type=MovieSceneScriptingFloatChannel, default=0, const [0,321]}
       CTRL_R_mouth_lowerLipDepress {type=MovieSceneScriptingFloatChannel, default=0, const [0,321]}
       CTRL_R_mouth_stretch {type=MovieSceneScriptingFloatChannel, default=0, const [0,321]}
       CTRL_L_mouth_towardsD {type=MovieSceneScriptingFloatChannel, default=0, const [0,321]}
       CTRL_R_mouth_funnelU {type=MovieSceneScriptingFloatChannel, default=0, const [0,321]}
       CTRL_L_mouth_towardsU {type=MovieSceneScriptingFloatChannel, default=0, const [0,321]}
       CTRL_R_mouth_lipBiteD {
        type=MovieSceneScriptingFloatChannel
        default=0
        thinned_from=321
        keys [[0,0.1971],[7,0.1964],[10,0.1949],[12,0.193],[16,0.1871],[21,0.1784],[26,0.1676],[28,0.1641],[29,0.1629],[32,0.1614],[44,0.1607],[49,0.1618],[51,0.1628],[53,0.1648],[60,0.1759],[62,0.1797],[66,0.1888],[68,0.1926],[69,0.1939],[72,0.196],[76,0.197],[101,0.1938],[104,0.1922],[107,0.1891],[117,0.1736],[126,0.1633],[132,0.1583],[145,0.1532],[159,0.1521],[179,0.1545],[188,0.1582],[192,0.1609],[197,0.1661],[204,0.1761],[211,0.1825],[218,0.1863],[238,0.1924],[248,0.1989],[269,0.2143],[288,0.2218],[300,0.221],[301,0.2203],[302,0.2188],[304,0.2145],[308,0.2043],[309,0.2023],[310,0.2009],[313,0.1984],[317,0.1972],[320,0.1971]]
       }
       CTRL_R_mouth_funnelD {type=MovieSceneScriptingFloatChannel, default=0, const [0,321]}
       CTRL_L_mouth_pressU {type=MovieSceneScriptingFloatChannel, default=0, const [0,321]}
       CTRL_L_mouth_funnelD {type=MovieSceneScriptingFloatChannel, default=0, const [0,321]}
       CTRL_L_mouth_suckBlow {type=MovieSceneScriptingFloatChannel, default=0, const [0,321]}
       CTRL_L_mouth_pressD {
        type=MovieSceneScriptingFloatChannel
        default=0
        thinned_from=321
        keys [[0,0.3697],[7,0.3683],[10,0.3655],[12,0.362],[16,0.3508],[21,0.3346],[26,0.3144],[28,0.3078],[29,0.3055],[32,0.3027],[45,0.3016],[50,0.3042],[52,0.307],[54,0.3115],[61,0.3332],[67,0.3578],[69,0.3637],[72,0.3675],[82,0.3697],[101,0.3635],[104,0.3605],[107,0.3546],[117,0.3255],[126,0.3062],[132,0.2968],[145,0.2874],[157,0.2851],[179,0.2898],[188,0.2966],[192,0.3018],[197,0.3116],[204,0.3303],[211,0.3423],[218,0.3494],[238,0.3608],[248,0.373],[269,0.402],[288,0.4159],[300,0.4144],[301,0.4131],[302,0.4104],[304,0.4023],[308,0.3832],[309,0.3794],[310,0.3767],[313,0.3721],[318,0.3698],[320,0.3698]]
       }
       CTRL_L_mouth_funnelU {type=MovieSceneScriptingFloatChannel, default=0, const [0,321]}
       CTRL_R_mouth_towardsU {type=MovieSceneScriptingFloatChannel, default=0, const [0,321]}
       CTRL_R_mouth_purseD {type=MovieSceneScriptingFloatChannel, default=0, const [0,321]}
       CTRL_L_mouth_lipBiteU {type=MovieSceneScriptingFloatChannel, default=0, const [0,321]}
       CTRL_R_mouth_pressD {
        type=MovieSceneScriptingFloatChannel
        default=0
        thinned_from=321
        keys [[0,0.37],[7,0.3686],[10,0.3658],[12,0.3623],[16,0.3511],[21,0.3348],[26,0.3146],[28,0.3081],[29,0.3058],[32,0.303],[45,0.3018],[50,0.3044],[52,0.3072],[54,0.3118],[61,0.3334],[67,0.3581],[69,0.364],[72,0.3678],[82,0.37],[101,0.3638],[104,0.3608],[107,0.3549],[117,0.3257],[126,0.3064],[132,0.297],[145,0.2876],[157,0.2853],[179,0.29],[188,0.2968],[192,0.302],[197,0.3118],[204,0.3306],[211,0.3426],[218,0.3497],[238,0.3611],[248,0.3733],[269,0.4023],[288,0.4162],[300,0.4147],[301,0.4134],[302,0.4108],[304,0.4027],[308,0.3835],[309,0.3797],[310,0.377],[313,0.3724],[318,0.3701],[320,0.3701]]
       }
       CTRL_R_mouth_pressU {type=MovieSceneScriptingFloatChannel, default=0, const [0,321]}
       CTRL_L_mouth_purseD {type=MovieSceneScriptingFloatChannel, default=0, const [0,321]}
       CTRL_R_mouth_lipBiteU {type=MovieSceneScriptingFloatChannel, default=0, const [0,321]}
       CTRL_R_mouth_suckBlow {type=MovieSceneScriptingFloatChannel, default=0, const [0,321]}
       CTRL_L_mouth_purseU {type=MovieSceneScriptingFloatChannel, default=0, const [0,321]}
       CTRL_R_mouth_towardsD {type=MovieSceneScriptingFloatChannel, default=0, const [0,321]}
       CTRL_L_mouth_lipBiteD {
        type=MovieSceneScriptingFloatChannel
        default=0
        thinned_from=321
        keys [[0,0.1971],[7,0.1964],[10,0.1949],[12,0.193],[16,0.1871],[21,0.1784],[26,0.1676],[28,0.1641],[29,0.1629],[32,0.1614],[44,0.1607],[49,0.1618],[51,0.1628],[53,0.1648],[60,0.1759],[62,0.1797],[66,0.1888],[68,0.1926],[69,0.1939],[72,0.196],[76,0.197],[101,0.1938],[104,0.1922],[107,0.1891],[117,0.1736],[126,0.1633],[132,0.1583],[145,0.1532],[159,0.1521],[179,0.1545],[188,0.1582],[192,0.1609],[197,0.1661],[204,0.1761],[211,0.1825],[218,0.1863],[238,0.1924],[248,0.1989],[269,0.2143],[288,0.2218],[300,0.221],[301,0.2203],[302,0.2188],[304,0.2145],[308,0.2043],[309,0.2023],[310,0.2009],[313,0.1984],[317,0.1972],[320,0.1971]]
       }
       CTRL_R_mouth_purseU {type=MovieSceneScriptingFloatChannel, default=0, const [0,321]}
       CTRL_R_mouth_cornerDepress {type=MovieSceneScriptingFloatChannel, default=0, const [0,321]}
       CTRL_R_mouth_upperLipRaise {
        type=MovieSceneScriptingFloatChannel
        default=0
        thinned_from=321
        keys [[0,0.2355],[7,0.2346],[10,0.2328],[12,0.2305],[16,0.2234],[21,0.2131],[26,0.2002],[28,0.1961],[29,0.1946],[32,0.1928],[45,0.1921],[50,0.1937],[52,0.1955],[54,0.1984],[61,0.2122],[67,0.2279],[69,0.2317],[72,0.2341],[83,0.2354],[102,0.2311],[105,0.2286],[109,0.2225],[117,0.2073],[126,0.195],[132,0.189],[145,0.183],[158,0.1816],[179,0.1846],[188,0.1889],[192,0.1922],[197,0.1985],[204,0.2104],[211,0.218],[218,0.2225],[238,0.2298],[248,0.2376],[269,0.256],[288,0.2649],[300,0.2639],[301,0.2631],[302,0.2614],[304,0.2563],[308,0.244],[309,0.2417],[310,0.24],[313,0.237],[317,0.2356],[320,0.2355]]
       }
       CTRL_R_mouth_dimple {type=MovieSceneScriptingFloatChannel, default=0, const [0,321]}
       CTRL_R_mouth_cornerPull {
        type=MovieSceneScriptingFloatChannel
        default=0
        thinned_from=321
        keys [[0,0.0721],[6,0.072],[11,0.071],[21,0.0653],[27,0.0607],[29,0.0596],[32,0.0591],[45,0.0588],[51,0.0596],[55,0.0613],[61,0.065],[67,0.0698],[69,0.071],[74,0.072],[85,0.0721],[102,0.0708],[105,0.07],[122,0.0613],[133,0.0577],[142,0.0564],[161,0.0556],[179,0.0565],[191,0.0586],[198,0.0613],[205,0.0649],[216,0.0679],[219,0.0683],[224,0.0687],[246,0.0722],[275,0.0794],[287,0.0811],[296,0.0812],[300,0.081],[302,0.0801],[304,0.0785],[308,0.0748],[310,0.0735],[314,0.0724],[320,0.0722]]
       }
      ]
     }]
   }
  ]
 }]
display_rate {denominator=1, numerator=30}
tick_resolution {denominator=1, numerator=24000}
tracks []
